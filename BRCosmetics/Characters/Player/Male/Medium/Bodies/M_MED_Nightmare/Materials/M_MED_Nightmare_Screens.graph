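# Lit skeletal mesh material for the screens on the Nightmare player body, showing animated, tinted scrolling glyphs in emissive over the standard body textures.
# The glyph texture tiles on the second UV channel and pans vertically, with per-region random noise and a sine varying the time, which is floored into discrete steps.
# A sphere mask on powered, centered UVs trims glyphs to a rounded-square shape.
# The assembled attributes pass through MF_TechArtOverrides.
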Material /BRCosmetics/Characters/Player/Male/Medium/Bodies/M_MED_Nightmare/Materials/M_MED_Nightmare_Screens
graph {
 properties {
  bUseMaterialAttributes=true
  bUsedWithMeshDeformer=true
  bUsedWithMorphTargets=true
  bUsedWithSkeletalMesh=true
  pixelDepthOffsetMode=PDOM_Legacy
  usedShadingModels=Default Lit
 }
 nodes [
  n1 = TextureSample() {samplerType=SAMPLERTYPE_Masks, texture=/BRCosmetics/Characters/Player/Male/Medium/Bodies/M_MED_Nightmare/Textures/T_Nightmare_Body_S}
  n2 = VectorParameter() {parameterName=GlyphUVScaleBias, defaultValue=<rgba 1.4 0.8 0 -0.069>}
  n3 = Add(A=Frac(None=TextureCoordinate() {coordinateIndex=1, uTiling=3, vTiling=0.1}), B=AppendVector(A=@n2.B, B=@n2.B))
  n4 = ComponentMask(None=@n2.RGB) {g=true, r=true}
  n5 = Time()
  n6 = MaterialFunctionCall!(
   PrimaryTextureCoordinates=TextureCoordinate() {coordinateIndex=1, uTiling=3}
   2ndOrderTextureCoordinates
   RandomSeed
   NoiseDensityBias=Constant() {r=0.4}
   Bypass Noise Influence Crosstalk
  ) {materialFunction=/Game/Effects/Fort_Effects/Materials/_Functions/MF_RandomUVSample}
  n7 = Add(A=Multiply(A=Sine(None=@n5) {period=18.5}, B=@n6.NoiseWithRange), B=@n5)
  n8 = Multiply(A=ScalarParameter() {defaultValue=0.278, parameterName=TimePanScale}, B=@n6.NoiseWithRange)
  n9 = ScalarParameter() {defaultValue=10, parameterName=TimeSteps}
  n10 = Floor(None=Multiply(A=Multiply(A=@n7, B=@n8), B=@n9))
  n11 = TextureSample(UVs=Panner(
    Coordinate=MaterialFunctionCall(UVs=@n3, Texture Scale=@n4) {materialFunction=&ScaleUVsByCenter}.UVs
    Time=Divide(A=@n10, B=@n9)
   ) {bFractionalPart=true, speedY=1}) {samplerType=SAMPLERTYPE_LinearColor, texture=/BRCosmetics/Characters/Player/Male/Medium/Bodies/M_MED_Nightmare/Textures/T_Nightmare_Glyphs}
  n12 = Multiply(A=@n11.R, B=VectorParameter() {parameterName=GlyphColor, defaultValue=<rgba 0.9 0.0141 0.0141 0>}.RGB)
  n13 = ConstantBiasScale(None=TextureCoordinate() {coordinateIndex=1}) {bias=-0.5, scale=2}
  n14 = SphereMask(A=Power(Base=Abs(None=@n13)) {constExponent=4}, B=Constant(), Radius=Constant() {r=0.7}, Hardness=Constant() {r=0.9})
  n15 = TextureSample() {samplerType=SAMPLERTYPE_Normal, texture=/BRCosmetics/Characters/Player/Male/Medium/Bodies/M_MED_Nightmare/Textures/T_Nightmare_Body_N}
  n16 = SetMaterialAttributes!(
   MaterialAttributes
   Base Color=NamedRerouteDeclaration(None=TextureSample() {texture=/BRCosmetics/Characters/Player/Male/Medium/Bodies/M_MED_Nightmare/Textures/T_Nightmare_Body_D}.RGB) {name=baseColor, nodeColor=<rgba 0.7109 1 0 1>}
   Metallic=NamedRerouteDeclaration(None=@n1.R) {name=metallic, nodeColor=<rgba 1 0.2109 0 1>}
   Specular=NamedRerouteDeclaration(None=@n1.R) {name=specular, nodeColor=<rgba 0 0.4375 1 1>}
   Roughness=NamedRerouteDeclaration(None=@n1.B) {name=roughness, nodeColor=<rgba 0 0.5781 1 1>}
   Emissive Color=NamedRerouteDeclaration(None=Multiply(A=@n12, B=@n14)) {name=emissive, nodeColor=<rgba 0.7109 0 1 1>}
   Normal=NamedRerouteDeclaration(None=@n15.RGB) {name=normal, nodeColor=<rgba 1 0.4453 0 1>}
  ) {attributeSetTypes [
    #1
    #2
    #3
    #4
    #5
    #6
   ]}
  MaterialFunctionCall(In=@n16) {materialFunction=&MF_TechArtOverrides}
 ]
}
