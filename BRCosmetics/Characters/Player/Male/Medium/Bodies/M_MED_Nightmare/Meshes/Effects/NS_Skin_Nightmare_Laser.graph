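# Niagara system for the Nightmare skin's red laser cosmetic, with three beams and flares attached to the character's skeletal mesh and faded by MasterAlpha.
# Laser spawns flickering mesh beams, aimed by blending camera and reticle vectors for the player's pawn, or by fixed mesh-relative vectors otherwise and in front-end previews.
# Laser_Flare brightens and enlarges sprites when facing the camera, using a dot product, and adds one light particle, killed based on its radius.
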NiagaraSystem /BRCosmetics/Characters/Player/Male/Medium/Bodies/M_MED_Nightmare/Meshes/Effects/NS_Skin_Nightmare_Laser
dependencies {
 usedDataInterfaces [&NiagaraDataInterfaceColorCurve, &NiagaraDataInterfaceArrayFloat, &NiagaraDataInterfaceCamera]
 usedDynamicInputs []
 usedModules [
  &self:SystemSpawnScript.SystemScriptSource.SystemScriptGraph.NiagaraNodeAssignment_0.SetVariables_#1
  &SystemState
  &self:GetCameraFacing
 ]
}
emitters {
 Laser {
  emitterUpdateScript [
   &EmitterState()
   &SpawnBurst_Instantaneous(
    Spawn Count=3
    Spawn Group:NiagaraInt32=&ReturnExecIndex()
    Spawn Probability=&MakeCustomFloatFromBool(
     Bool=&SetBoolByIntComparison(A=$Emitter.LoopCount, B=0, Comparison Type=::NewEnumerator4)
     True Float=1
     False Float=0
    )
    Use Spawn Probability=true
   )
  ]
  particleSpawnScript [
   /Niagara/Modules/Spawn/Initialization/InitializeParticle(
    Color=<rgba 10 0 0 1>
    Lifetime=0
    Mesh Scale=&RandomRangeVector(Minimum=<0.15 1 1>, Maximum=<0.2 1 1>)
    Module.Write Color=true
    Module.Write Lifetime=false
    Module.Write Mass=false
    Module.Write Material Random=true
    Module.Write Position=false
    Module.Write Scale=true
    Module.Write SpriteRotation=false
    Module.Write SpriteSize=false
    Position=&SimulationPosition()
    Sprite Size=<5 5>
   )
   /SetVariables(Particles.Offset:Vector3f=&self:PickOffsetByExecIndex()) {isSetParameters=true}
  ]
  particleUpdateScript [
   &ParticleState(Kill Particles When Lifetime Has Elapsed=false, Let Infinitely Lived Particles Die When Emitter Deactivates=true)
   &DynamicMaterialParameters(
    Index 0 Param 1=&FN_FrontEndFloatModifier(Gameplay=0.3, Front End=0.5)
    Index 0 Param 2=&RandomRangeFloat(Minimum=0, Maximum=1)
    Index 0 Param 3=&Lerp_Float(
     A=&RandomRangeFloat(Minimum=0, Maximum=1)
     B=&RandomRangeFloat(Minimum=-1, Maximum=1)
     Alpha=$Particles.NormalizedAge
    )
    Index 0 Param 4=1
    Module.Param0WriteEnabled=true
    Module.Param1WriteEnabled=true
    Module.Param2WriteEnabled=true
    Module.Param3WriteEnabled=true
   )
   &SkeletalMeshLocation(
    Inherit Velocity=false
    Interpolate Mesh Sampling=false
    MaskSpawnGroup=false
    Mesh Sampling Type=::NewEnumerator5
    Sampled Position Offset=$Particles.Offset
    Sampled Position Offset Coordinate Space:ENiagaraMeshTransforms=::NewEnumerator3
    Skeletal Mesh=$User.New Skeletal Mesh
    Socket Sampling Mode=::NewEnumerator5
   )
   &SkeletalMeshLocation(
    Apply To Particle Position=false
    Bone Sampling Mode=::NewEnumerator4
    Inherit Velocity=false
    Interpolate Mesh Sampling=false
    MaskSpawnGroup=false
    Sampled Position Offset=$Particles.Offset
    Sampled Position Offset Coordinate Space:ENiagaraMeshTransforms=::NewEnumerator3
    Skeletal Mesh=$User.New Skeletal Mesh
   ) {module=SkeletalMeshLocation001}
   &ScaleColor(
    Scale Alpha=&Multiply_Float(
     A=$User.MasterAlpha
     B=&self:Interpolate_Float(
      Float to Interpolate:NiagaraFloat=&RandomRangeFloat(Minimum=0, Maximum=1, Evaluation Type=::NewEnumerator1, Fixed Random Seed=true)
      Rate of Change:NiagaraFloat=20
     )
    )
    ScaleA=true
    ScaleRGB=false
   )
   /SetVariables(
    Particles.AdjustHighLowAim:NiagaraFloat=&ClampFloat(
     Float=&Add_Float(
      A=-0.5
      B=&Abs_Float(Float:NiagaraFloat=&MakeFloatFromVector(Vector=$System.Forward Vector World, Channel=::NewEnumerator2))
     )
     Max=0.5
    )
    Particles.AimToCamera:Vector3f=&TransformVector(Vector=$System.Forward Vector World, Source Space=::World, Destination Space=::Local)
    Particles.AimToReticle:Vector3f=&RotateVectorByQuaternion(Vector To Rotate=<0 1 -0.24>, Quaternion=$Output.SkeletalMeshLocation001.SampledRotation)
    Particles.Scale=/Niagara/DynamicInputs/Math/Clamp_Vector(
     Vector:Vector3f=&self:Interpolate_Vector(
      Initial Value:Vector3f=$Particles.Initial.Scale
      Rate Of Change:NiagaraFloat=10
      Vector to Interpolate:Vector3f=&Add_Vector(
       A=$Particles.Scale
       B=&RandomRangeVector(Minimum=<-0.005 0 0>, Maximum=<0.005 0 0>, Evaluation Type=::NewEnumerator1)
      )
     )
     Min:Vector3f=<0.04 0 0>
     Max:Vector3f=<0.2 1 1>
    )
    Particles.SimpleAiming:Vector3f=&RotateVectorByQuaternion(Vector To Rotate=<-0.1 1 -0.56>, Quaternion=$Output.SkeletalMeshLocation001.SampledRotation)
    Particles.UseDynamicAiming:NiagaraBool=&Boolean_AndOperation(Bool1=$User.bIsPlayersPawn, Bool2=&Boolean_NotOperation(Bool=$User.bIsFrontEndPreview))
   ) {isSetParameters=true}
   /SetVariables(Particles.MeshOrientation=&OrientMeshToVector(
     Look At Direction=&MakeCustomVectorFromBool(
      Bool=$Particles.UseDynamicAiming
      True Vector=&Lerp_Vector(A=$Particles.AimToReticle, B=$Particles.AimToCamera, Alpha=&Add_Float(A=0.5, B=$Particles.AdjustHighLowAim))
      False Vector=&FN_FrontEndVector3Modifier(
       Gameplay=$Particles.SimpleAiming
       Front End=&RotateVectorByQuaternion(Vector To Rotate=<1 0 0>, Quaternion=$Output.SkeletalMeshLocation.SampledRotation)
      )
     )
     Reference Axis Vector=<1 0 0>
    )) {isSetParameters=true}
   &Color(Color=&ColorFromCurve(
     ColorCurve=DataInterface() {propertyValues {
       AlphaCurve {keys [<at 0 1 RCIM_Cubic>]}
       BlueCurve {keys [<at 0 0 RCIM_Cubic>,<at 0.3333 0>,<at 0.6667 0>,<at 1 1 RCIM_Cubic>]}
       GreenCurve {keys [<at 0 0 RCIM_Cubic>,<at 0.3333 1>,<at 0.6667 1>,<at 1 0 RCIM_Cubic>]}
       RedCurve {keys [<at 0 1 RCIM_Cubic>,<at 0.3333 1>,<at 0.6667 0>,<at 1 0 RCIM_Cubic>]}
      }}
     CurveIndex=&ClampFloat(
      Float=&Add_Float(
       A=-0.5
       B=&Abs_Float(Float:NiagaraFloat=&MakeFloatFromVector(Vector=$System.Forward Vector World, Channel=::NewEnumerator2))
      )
      Max=0.5
     )
    )) {enabled=false}
  ]
  rendererClasses [&NiagaraMeshRendererProperties]
  renderers [{rendererClass=&NiagaraMeshRendererProperties, rendererIndex=0}]
 }
 Laser_Flare {
  emitterUpdateScript [
   &EmitterState()
   &SpawnBurst_Instantaneous(
    Spawn Count=3
    Spawn Probability=&MakeCustomFloatFromBool(
     Bool=&SetBoolByIntComparison(A=$Emitter.LoopCount, B=0, Comparison Type=::NewEnumerator4)
     True Float=1
     False Float=0
    )
    Use Spawn Probability=true
   )
   &SpawnBurst_Instantaneous(
    Spawn Count=1
    Spawn Group=1
    Spawn Probability=&MakeCustomFloatFromBool(
     Bool=&SetBoolByIntComparison(A=$Emitter.LoopCount, B=0, Comparison Type=::NewEnumerator4)
     True Float=1
     False Float=0
    )
    Use Spawn Probability=true
   ) {module=SpawnBurst_Instantaneous001}
  ]
  particleSpawnScript [
   &InitializeParticle(
    Color=<rgba 1 0 0 1>
    Lifetime=0
    Module.Write Color=true
    Module.Write Lifetime=false
    Module.Write Mass=true
    Module.Write Material Random=true
    Module.Write Position=false
    Module.Write Scale=false
    Module.Write SpriteRotation=true
    Module.Write SpriteSize=true
    Position=&SimulationPosition()
    Sprite Rotation=45
    Sprite Size=<7 7>
   )
   &CameraOffset(Camera Offset Amount=1)
   &DynamicMaterialParameters(
    Index 0 Param 1=0.5
    Index 0 Param 2=1
    Index 0 Param 3=1
    Index 0 Param 4=1
    Index 1 Param 1=5
    Module.Param0WriteEnabled=true
    Module.Param0WriteEnabled1=true
    Module.Param1WriteEnabled=true
    Module.Param1WriteEnabled1=true
    Module.Param2WriteEnabled=true
    Module.Param2WriteEnabled1=true
    Module.Param3WriteEnabled=true
    Module.Param3WriteEnabled1=true
    Write Parameter Index 1=true
   )
   /SetVariables(Particles.Offset:Vector3f=&MakeCustomVectorFromBool(
     Bool=&SetBoolByIntComparison(A=&SpawnGroup(), B=1, Comparison Type=::NewEnumerator4)
     True Vector=$System.LightOffset
     False Vector=&self:PickOffsetByExecIndex()
    )) {isSetParameters=true}
  ]
  particleUpdateScript [
   &ParticleState(Kill Particles When Lifetime Has Elapsed=false, Let Infinitely Lived Particles Die When Emitter Deactivates=true)
   &SkeletalMeshLocation(
    Inherit Velocity=false
    Interpolate Mesh Sampling=false
    MaskSpawnGroup=false
    Mesh Sampling Type=::NewEnumerator5
    Sampled Position Offset=$Particles.Offset
    Sampled Position Offset Coordinate Space:ENiagaraMeshTransforms=::NewEnumerator3
    Skeletal Mesh=$User.New Skeletal Mesh
    Socket Sampling Mode=::NewEnumerator5
   )
   &SkeletalMeshLocation(
    Apply To Particle Position=false
    Bone Sampling Mode=::NewEnumerator4
    Inherit Velocity=false
    Interpolate Mesh Sampling=false
    MaskSpawnGroup=false
    Sampled Position Offset=$Particles.Offset
    Sampled Position Offset Coordinate Space:ENiagaraMeshTransforms=::NewEnumerator3
    Skeletal Mesh=$User.New Skeletal Mesh
   ) {module=SkeletalMeshLocation001}
   &self:DotCameraInfo(
    Dot Vector:Vector3f=&RotateVectorByQuaternion(
     Vector To Rotate=&FN_FrontEndVector3Modifier(Gameplay=<-0.06 1 -0.49>, Front End=<0.11 1 0>)
     Quaternion=$Output.SkeletalMeshLocation001.SampledRotation
    )
    Dot product to return 0=-0.99
    Dot product to return 1=-1
    Interpolate Value=5
   ) {module=DotCameraInfo}
   &ScaleColor(
    Scale Alpha=&Multiply_Float(
     A=$User.MasterAlpha
     B=&SelectFloatFromArray(Direct Array Index=&SpawnGroup(), Float Selection Array=DataInterface() {propertyValues {FloatData [1,0]}})
    )
    Scale RGB=&VectorFromFloat(Value=&Lerp_Float(A=1, B=10, Alpha=$Particles.CameraVectorDot))
    ScaleA=true
    ScaleRGB=true
   )
   &ScaleSpriteSize(Scale Factor=&Vector2DFromFloat(Value=&Lerp_Float(A=1, B=1.25, Alpha=$Particles.CameraVectorDot)))
   &FN_Light_Attributes(
    Color=$Particles.Initial.Color
    Create Color Attribute=true
    Create Exponent Attribute=true
    Create Radius Attribute=true
    Exponent=1
    Radius=&Multiply_Float(
     A=&SelectFloatFromArray(Direct Array Index=&SpawnGroup(), Float Selection Array=DataInterface() {propertyValues {FloatData [0,15]}})
     B=$User.MasterAlpha
    )
    Scale Color=0.5
   )
   &KillParticles(Kill Particles=&Boolean_AndOperation(
     Bool1=&SetBoolByIntComparison(A=&SpawnGroup(), B=1, Comparison Type=::NewEnumerator4)
     Bool2=&SetBoolByFloatComparison(A=$Particles.FN_Light_Attributes.Radius, B=0, Comparison Type=::NewEnumerator7)
    ))
   &Color(Color=&ColorFromCurve(
     ColorCurve=DataInterface() {propertyValues {
       AlphaCurve {keys [<at 0 1 RCIM_Cubic>]}
       BlueCurve {keys [<at 0 0 RCIM_Cubic>,<at 0.3333 0>,<at 0.6667 0>,<at 1 1 RCIM_Cubic>]}
       GreenCurve {keys [<at 0 0 RCIM_Cubic>,<at 0.3333 1>,<at 0.6667 1>,<at 1 0 RCIM_Cubic>]}
       RedCurve {keys [<at 0 1 RCIM_Cubic>,<at 0.3333 1>,<at 0.6667 0>,<at 1 0 RCIM_Cubic>]}
      }}
     CurveIndex=$Particles.CameraVectorDot
    )) {enabled=false}
   /SetVariables(Particles.TestSpriteSize:Vector2f=&Vector2DFromFloat(Value=1)) {isSetParameters=true}
  ]
  rendererClasses [&NiagaraSpriteRendererProperties, &NiagaraLightRendererProperties]
  renderers [
   {rendererClass=&NiagaraSpriteRendererProperties}
   {rendererClass=&NiagaraLightRendererProperties, rendererIndex=1}
  ]
 }
}
summary {
 systemName=NS_Skin_Nightmare_Laser
 emitters [
  {emitterName=Laser, rendererClasses [&NiagaraMeshRendererProperties]}
  {emitterName=Laser_Flare, rendererClasses [&NiagaraSpriteRendererProperties, &NiagaraLightRendererProperties]}
 ]
 userVariables [
  User.bIsFrontEnd {type=NiagaraBool}
  User.bIsFrontEndPreview {type=NiagaraBool}
  User.bIsPlayersPawn {type=NiagaraBool}
  User.MasterAlpha {type=NiagaraFloat, defaultValue=1, description=""}
  User.New Skeletal Mesh {
   type=NiagaraDataInterfaceSkeletalMesh
   defaultValue {
    struct=NiagaraExt_VariableValue_DataInterface
    value {dataInterface=&self:NiagaraDataInterfaceSkeletalMesh_0}
   }
  }
 ]
}
