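# Physics asset for the medium male Nightmare player body, covering its dynamic accessory bones: a left cord, a skirt piece and a necklace.
# Tiny MSTR capsules act as anchors for narrowly limited joints; the cord and necklace anchors link to spine_05, while the skirt anchor has no further constraint.
# Two three-link necklace chains close into a loop through an angularly free joint, with a pendant on the right end.
# Large spine_05 and neck_01 capsules provide torso and neck volumes.
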PhysicsAsset /Game/Characters/Player/Male/Medium/Bodies/M_MED_Nightmare/Meshes/M_MED_Nightmare_Physics
bodies {
 dyn_cord_01_l [{shapeType=Capsule, length=5.1992, radius=1.9157, center=<3.0249 0 0>, rotation=<rot -90 -0.0199 0.02>}]
 dyn_cord_01_l_MSTR [{shapeType=Capsule, length=0.01, radius=0.01}]
 dyn_necklaceBone_00 [{shapeType=Capsule, length=4.7711, radius=1.6111, center=<4.6175 0 -0.0005>, rotation=<rot 90 -0.0033 -0.0032>}]
 dyn_necklace_01_l [{shapeType=Capsule, length=3.5165, radius=0.8475, center=<3.014 0 -0.0007>, rotation=<rot 90 -0.0033 -0.0032>}]
 dyn_necklace_01_l_MSTR [{shapeType=Capsule, length=0.01, radius=0.01}]
 dyn_necklace_01_r [{shapeType=Capsule, length=3.5165, radius=0.8475, center=<3.014 0 -0.0007>, rotation=<rot 90 -0.0033 -0.0032>}]
 dyn_necklace_01_r_MSTR [{shapeType=Capsule, length=0.01, radius=0.01}]
 dyn_necklace_02_l [{shapeType=Capsule, length=3.5165, radius=0.8475, center=<3.014 0 -0.0007>, rotation=<rot 90 -0.0033 -0.0032>}]
 dyn_necklace_02_r [{shapeType=Capsule, length=3.5165, radius=0.8475, center=<3.014 0 -0.0007>, rotation=<rot 90 -0.0033 -0.0032>}]
 dyn_necklace_03_l [{shapeType=Capsule, length=3.5165, radius=0.8475, center=<3.014 0 -0.0007>, rotation=<rot 90 -0.0033 -0.0032>}]
 dyn_necklace_03_r [{shapeType=Capsule, length=3.5165, radius=0.8475, center=<3.014 0 -0.0007>, rotation=<rot 90 -0.0033 -0.0032>}]
 dyn_skirt_01 [{shapeType=Capsule, length=11.1211, radius=3.7728, center=<5.2435 -0.0001 0>, rotation=<rot -90 0.0132 -0.0131>}]
 dyn_skirt_01_MSTR [{shapeType=Capsule, length=0.01, radius=0.01}]
 neck_01 [{shapeType=Capsule, length=5.613, radius=6.7406, center=<2.3175 0 0>, rotation=<rot 0 -89.9999 89.9996>}]
 spine_05 [{shapeType=Capsule, length=21.8443, radius=21.526, center=<-9.5575 -6.4235 0>}]
}
constraints [
 {bone1Name=dyn_cord_01_l, bone2Name=dyn_cord_01_l_MSTR, swing1LimitDegrees=20, swing2LimitDegrees=5, twistLimitDegrees=5}
 {bone1Name=dyn_skirt_01, bone2Name=dyn_skirt_01_MSTR, swing1LimitDegrees=5, swing2LimitDegrees=45, twistLimitDegrees=5}
 {bone1Name=dyn_necklace_01_l, bone2Name=dyn_necklace_01_l_MSTR, swing1LimitDegrees=5, swing2LimitDegrees=5, twistLimitDegrees=5}
 {bone1Name=dyn_necklace_02_l, bone2Name=dyn_necklace_01_l, swing1LimitDegrees=15, swing2LimitDegrees=15, twistLimitDegrees=5}
 {bone1Name=dyn_necklace_03_l, bone2Name=dyn_necklace_02_l, swing1LimitDegrees=15, swing2LimitDegrees=15, twistLimitDegrees=5}
 {bone1Name=dyn_necklace_01_r, bone2Name=dyn_necklace_01_r_MSTR, swing1LimitDegrees=5, swing2LimitDegrees=5, twistLimitDegrees=5}
 {bone1Name=dyn_necklace_02_r, bone2Name=dyn_necklace_01_r, swing1LimitDegrees=15, swing2LimitDegrees=15, twistLimitDegrees=5}
 {bone1Name=dyn_necklace_03_r, bone2Name=dyn_necklace_02_r, swing1LimitDegrees=15, swing2LimitDegrees=15, twistLimitDegrees=5}
 {bone1Name=dyn_necklaceBone_00, bone2Name=dyn_necklace_03_r, swing1LimitDegrees=15, swing2LimitDegrees=15, twistLimitDegrees=5}
 {
  bone1Name=dyn_necklace_03_l
  bone2Name=dyn_necklace_03_r
  swing1LimitDegrees=45
  swing1Motion=Free
  swing2LimitDegrees=45
  swing2Motion=Free
  twistLimitDegrees=45
  twistMotion=Free
 }
 {
  bone1Name=dyn_cord_01_l_MSTR
  bone2Name=spine_05
  swing1LimitDegrees=45
  swing1Motion=Free
  swing2LimitDegrees=45
  swing2Motion=Free
  twistLimitDegrees=45
  twistMotion=Free
 }
 {
  bone1Name=dyn_necklace_01_l_MSTR
  bone2Name=spine_05
  swing1LimitDegrees=45
  swing1Motion=Free
  swing2LimitDegrees=45
  swing2Motion=Free
  twistLimitDegrees=45
  twistMotion=Free
 }
 {
  bone1Name=dyn_necklace_01_r_MSTR
  bone2Name=spine_05
  swing1LimitDegrees=45
  swing1Motion=Free
  swing2LimitDegrees=45
  swing2Motion=Free
  twistLimitDegrees=45
  twistMotion=Free
 }
 {
  bone1Name=neck_01
  bone2Name=spine_05
  swing1LimitDegrees=45
  swing1Motion=Free
  swing2LimitDegrees=45
  swing2Motion=Free
  twistLimitDegrees=45
  twistMotion=Free
 }
]
