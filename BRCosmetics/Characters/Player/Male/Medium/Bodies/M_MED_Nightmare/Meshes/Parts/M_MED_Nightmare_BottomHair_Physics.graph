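# Physics asset for the bottom hair part of the medium male Nightmare player cosmetic body, setting up its dangling braids for physics.
# Four braids, two left and two right, are four-capsule chains, each hung from a tiny MSTR root body by tightly limited swing and twist constraints.
# Larger capsules on head, spine_05 and left upper arm give body volumes, with head and arm linked to the spine by angularly free constraints.
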PhysicsAsset /BRCosmetics/Characters/Player/Male/Medium/Bodies/M_MED_Nightmare/Meshes/Parts/M_MED_Nightmare_BottomHair_Physics
bodies {
 dyn_braidinner_A_00_r [{shapeType=Capsule, length=4.9272, radius=1.734, center=<3.5036 0 0.0001>, rotation=<rot -90 10.025 -10.025>}]
 dyn_braidinner_A_00_r_MSTR [{shapeType=Capsule, length=0.01, radius=0.01}]
 dyn_braidinner_A_01_r [{shapeType=Capsule, length=4.6239, radius=1.3446, center=<3.5036 0 0>, rotation=<rot -90 10.025 -10.0251>}]
 dyn_braidinner_A_02_r [{shapeType=Capsule, length=4.6239, radius=1.3446, center=<3.5036 0 0>, rotation=<rot -90 54.7356 -54.7356>}]
 dyn_braidinner_A_03_r [{shapeType=Capsule, length=4.4045, radius=1.734, center=<3.5036 0 0>, rotation=<rot -90 -19.471 19.4707>}]
 dyn_braidinner_C_00_r [{shapeType=Capsule, length=3.6324, radius=3.0288, center=<3.5036 0 0.0001>, rotation=<rot -90 35.2645 -35.2646>}]
 dyn_braidinner_C_00_r_MSTR [{shapeType=Capsule, length=0.01, radius=0.01}]
 dyn_braidinner_C_01_r [{shapeType=Capsule, length=5.3532, radius=2.0093, center=<3.2354 0 0>, rotation=<rot -90 -5.0511 5.0509>}]
 dyn_braidinner_C_02_r [{shapeType=Capsule, length=3.5549, radius=2.0093, center=<3.5054 0 0>, rotation=<rot -90 5.0512 -5.0513>}]
 dyn_braidinner_C_03_r [{shapeType=Capsule, length=5.2197, radius=2.0386, center=<4.2646 0 0>, rotation=<rot -90 35.2645 -35.2646>}]
 dyn_braidinner_F_00_l [{shapeType=Capsule, length=3.6324, radius=3.0288, center=<3.5036 0 0.0001>, rotation=<rot -90 35.2644 -35.2643>}]
 dyn_braidinner_F_00_l_MSTR [{shapeType=Capsule, length=0.01, radius=0.01}]
 dyn_braidinner_F_01_l [{shapeType=Capsule, length=4.8532, radius=2.2593, center=<4 0 0>, rotation=<rot -90 0 0>}]
 dyn_braidinner_F_02_l [{shapeType=Capsule, length=3.5549, radius=2.2593, center=<4 0 0>, rotation=<rot -90 0 0>}]
 dyn_braidinner_F_03_l [{shapeType=Capsule, length=5.2197, radius=2.0386, center=<4 0 0>, rotation=<rot -90 0 0>}]
 dyn_braidsmall_00_l [{shapeType=Capsule, length=3.1097, radius=3.0288, center=<3.5036 0 0>, rotation=<rot -90 25.2394 -25.2393>}]
 dyn_braidsmall_00_l_MSTR [{shapeType=Capsule, length=0.01, radius=0.01}]
 dyn_braidsmall_01_l [{shapeType=Capsule, length=5.3532, radius=2.0093, center=<3.2354 0 0>, rotation=<rot -90 -6.7215 6.7212>}]
 dyn_braidsmall_02_l [{shapeType=Capsule, length=3.5549, radius=2.0093, center=<3.5054 0 0>, rotation=<rot -90 13.2627 -13.2626>}]
 dyn_braidsmall_03_l [{shapeType=Capsule, length=5.2197, radius=2.0386, center=<4.2646 0 0>, rotation=<rot -90 54.7356 -54.7355>}]
 head [{shapeType=Capsule, length=12.0739, radius=9.3297, rotation=<rot 90 0.0264 0.0264>}]
 spine_05 [
  {shapeType=Capsule, length=28.1628, radius=8.9064, center=<4.6301 -2.624 0>}
  {shapeType=Capsule, length=10.0553, radius=7.4979, center=<8.8415 -3.5847 -19.8597>}
  {shapeType=Capsule, length=10.0553, radius=7.4979, center=<5.3888 -1.1883 19.0064>}
  {shapeType=Capsule, length=19.4197, radius=10.5803, center=<0.5385 -4.0682 -0.0001>}
 ]
 upperarm_l [{shapeType=Capsule, length=15.5232, radius=9.3901, center=<4.769 0 4.4524>}]
}
constraints [
 {bone1Name=dyn_braidinner_C_00_r, bone2Name=dyn_braidinner_C_00_r_MSTR, swing1LimitDegrees=20, swing2LimitDegrees=5, twistLimitDegrees=5}
 {bone1Name=dyn_braidinner_C_01_r, bone2Name=dyn_braidinner_C_00_r, swing1LimitDegrees=20, swing2LimitDegrees=15, twistLimitDegrees=5}
 {bone1Name=dyn_braidinner_C_02_r, bone2Name=dyn_braidinner_C_01_r, swing1LimitDegrees=20, swing2LimitDegrees=15, twistLimitDegrees=5}
 {bone1Name=dyn_braidinner_C_03_r, bone2Name=dyn_braidinner_C_02_r, swing1LimitDegrees=20, swing2LimitDegrees=15, twistLimitDegrees=5}
 {bone1Name=dyn_braidsmall_00_l, bone2Name=dyn_braidsmall_00_l_MSTR, swing1LimitDegrees=5, swing2LimitDegrees=5, twistLimitDegrees=5}
 {bone1Name=dyn_braidsmall_01_l, bone2Name=dyn_braidsmall_00_l, swing1LimitDegrees=20, swing2LimitDegrees=20, twistLimitDegrees=5}
 {bone1Name=dyn_braidsmall_02_l, bone2Name=dyn_braidsmall_01_l, swing1LimitDegrees=20, swing2LimitDegrees=20, twistLimitDegrees=5}
 {bone1Name=dyn_braidsmall_03_l, bone2Name=dyn_braidsmall_02_l, swing1LimitDegrees=20, swing2LimitDegrees=20, twistLimitDegrees=5}
 {bone1Name=dyn_braidinner_F_00_l, bone2Name=dyn_braidinner_F_00_l_MSTR, swing1LimitDegrees=5, swing2LimitDegrees=5, twistLimitDegrees=5}
 {bone1Name=dyn_braidinner_F_01_l, bone2Name=dyn_braidinner_F_00_l, swing1LimitDegrees=15, swing2LimitDegrees=25, twistLimitDegrees=5}
 {bone1Name=dyn_braidinner_F_02_l, bone2Name=dyn_braidinner_F_01_l, swing1LimitDegrees=15, swing2LimitDegrees=25, twistLimitDegrees=5}
 {bone1Name=dyn_braidinner_F_03_l, bone2Name=dyn_braidinner_F_02_l, swing1LimitDegrees=15, swing2LimitDegrees=25, twistLimitDegrees=5}
 {bone1Name=dyn_braidinner_A_00_r, bone2Name=dyn_braidinner_A_00_r_MSTR, swing1LimitDegrees=5, swing2LimitDegrees=5, twistLimitDegrees=5}
 {bone1Name=dyn_braidinner_A_01_r, bone2Name=dyn_braidinner_A_00_r, swing1LimitDegrees=20, swing2LimitDegrees=20, twistLimitDegrees=5}
 {bone1Name=dyn_braidinner_A_02_r, bone2Name=dyn_braidinner_A_01_r, swing1LimitDegrees=20, swing2LimitDegrees=20, twistLimitDegrees=5}
 {bone1Name=dyn_braidinner_A_03_r, bone2Name=dyn_braidinner_A_02_r, swing1LimitDegrees=20, swing2LimitDegrees=20, twistLimitDegrees=5}
 {
  bone1Name=upperarm_l
  bone2Name=spine_05
  swing1LimitDegrees=45
  swing1Motion=Free
  swing2LimitDegrees=45
  swing2Motion=Free
  twistLimitDegrees=45
  twistMotion=Free
 }
 {
  bone1Name=head
  bone2Name=spine_05
  swing1LimitDegrees=45
  swing1Motion=Free
  swing2LimitDegrees=45
  swing2Motion=Free
  twistLimitDegrees=45
  twistMotion=Free
 }
]
