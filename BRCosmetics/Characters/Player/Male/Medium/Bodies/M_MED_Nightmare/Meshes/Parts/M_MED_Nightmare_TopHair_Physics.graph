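# Physics asset for the M_MED_Nightmare male cosmetic's top hair, defining capsule bodies and joints for nine braid chains (D to G on each side, plus a middle one).
# Each four-segment chain starts at a tiny MSTR anchor body, constrained with free rotation to neck_02 for side braids and to head for the middle one.
# Tight twist and per-joint swing limits shape the braids, except the free-rotating middle root.
# Head, neck, upper spine and upper-arm capsules approximate the body.
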PhysicsAsset /BRCosmetics/Characters/Player/Male/Medium/Bodies/M_MED_Nightmare/Meshes/Parts/M_MED_Nightmare_TopHair_Physics
bodies {
 dyn_braidouter_D_00_l [{shapeType=Capsule, length=3.6324, radius=3.0288, center=<3.5036 0 0>, rotation=<rot -90 0 0>}]
 dyn_braidouter_D_00_l_MSTR [{shapeType=Capsule, length=0.01, radius=0.01}]
 dyn_braidouter_D_00_r [{shapeType=Capsule, length=3.6324, radius=3.0288, center=<3.5036 0 0>, rotation=<rot -90 0 0>}]
 dyn_braidouter_D_00_r_MSTR [{shapeType=Capsule, length=0.01, radius=0.01}]
 dyn_braidouter_D_01_l [{shapeType=Capsule, length=5.6565, radius=2.3987, center=<3.2354 0 0>, rotation=<rot -90 0 0>}]
 dyn_braidouter_D_01_r [{shapeType=Capsule, length=3.5126, radius=1.8197, center=<3.2354 0 0>, rotation=<rot -90 0 0>}]
 dyn_braidouter_D_02_l [{shapeType=Capsule, length=3.8582, radius=2.3987, center=<3.5054 0 0>, rotation=<rot -90 0 0>}]
 dyn_braidouter_D_02_r [{shapeType=Capsule, length=3.5961, radius=2.0097, center=<3.5054 0 0>, rotation=<rot -90 0 0>}]
 dyn_braidouter_D_03_l [{shapeType=Capsule, length=6.2695, radius=1.5114, center=<4.2646 0 0>, rotation=<rot -90 0 0>}]
 dyn_braidouter_D_03_r [{shapeType=Capsule, length=6.2695, radius=1.5114, center=<4.2646 0 0>, rotation=<rot -90 0 0>}]
 dyn_braidouter_E_00_l [{shapeType=Capsule, length=3.6324, radius=3.0288, center=<3.5036 0 0>, rotation=<rot -90 0 0>}]
 dyn_braidouter_E_00_l_MSTR [{shapeType=Capsule, length=0.01, radius=0.01}]
 dyn_braidouter_E_00_r [{shapeType=Capsule, length=3.6324, radius=3.0288, center=<3.5036 0 0>, rotation=<rot -90 0 0>}]
 dyn_braidouter_E_00_r_MSTR [{shapeType=Capsule, length=0.01, radius=0.01}]
 dyn_braidouter_E_01_l [{shapeType=Capsule, length=5.6565, radius=2.3987, center=<3.2354 0 0>, rotation=<rot -90 0 0>}]
 dyn_braidouter_E_01_r [{shapeType=Capsule, length=5.6565, radius=2.3987, center=<3.2354 0 0>, rotation=<rot -90 0 0>}]
 dyn_braidouter_E_02_l [{shapeType=Capsule, length=3.8582, radius=2.3987, center=<3.5054 0 0>, rotation=<rot -90 0 0>}]
 dyn_braidouter_E_02_r [{shapeType=Capsule, length=3.8582, radius=2.3987, center=<3.5054 0 0>, rotation=<rot -90 0 0>}]
 dyn_braidouter_E_03_l [{shapeType=Capsule, length=6.2695, radius=1.5114, center=<4.2646 0 0>, rotation=<rot -90 0 0>}]
 dyn_braidouter_E_03_r [{shapeType=Capsule, length=6.2695, radius=1.5114, center=<4.2646 0 0>, rotation=<rot -90 0 0>}]
 dyn_braidouter_F_00_l [{shapeType=Capsule, length=3.6324, radius=3.0288, center=<3.5036 0 0>, rotation=<rot -90 0 0>}]
 dyn_braidouter_F_00_l_MSTR [{shapeType=Capsule, length=0.01, radius=0.01}]
 dyn_braidouter_F_00_r [{shapeType=Capsule, length=3.6324, radius=3.0288, center=<3.5036 0 0>, rotation=<rot -90 0 0>}]
 dyn_braidouter_F_00_r_MSTR [{shapeType=Capsule, length=0.01, radius=0.01}]
 dyn_braidouter_F_01_l [{shapeType=Capsule, length=5.6565, radius=2.3987, center=<3.2354 0 0>, rotation=<rot -90 0 0>}]
 dyn_braidouter_F_01_r [{shapeType=Capsule, length=5.6565, radius=2.3987, center=<3.2354 0 0>, rotation=<rot -90 0 0>}]
 dyn_braidouter_F_02_l [{shapeType=Capsule, length=3.8582, radius=2.3987, center=<3.5054 0 0>, rotation=<rot -90 0 0>}]
 dyn_braidouter_F_02_r [{shapeType=Capsule, length=3.8582, radius=2.3987, center=<3.5054 0 0>, rotation=<rot -90 0 0>}]
 dyn_braidouter_F_03_l [{shapeType=Capsule, length=6.2695, radius=1.5114, center=<4.2646 0 0>, rotation=<rot -90 0 0>}]
 dyn_braidouter_F_03_r [{shapeType=Capsule, length=6.2695, radius=1.5114, center=<4.2646 0 0>, rotation=<rot -90 0 0>}]
 dyn_braidouter_G_00_l [{shapeType=Capsule, length=3.6324, radius=3.0288, center=<3.5036 0 0>, rotation=<rot -90 0 0>}]
 dyn_braidouter_G_00_l_MSTR [{shapeType=Capsule, length=0.01, radius=0.01}]
 dyn_braidouter_G_00_r [{shapeType=Capsule, length=3.6324, radius=3.0288, center=<3.5036 0 0>, rotation=<rot -90 174.9489 -174.9488>}]
 dyn_braidouter_G_00_r_MSTR [{shapeType=Capsule, length=0.01, radius=0.01}]
 dyn_braidouter_G_01_l [{shapeType=Capsule, length=5.6565, radius=2.3987, center=<3.2354 0 0>, rotation=<rot -90 0 0>}]
 dyn_braidouter_G_01_r [{shapeType=Capsule, length=5.6565, radius=2.3987, center=<3.2354 0 0>, rotation=<rot -90 0 0>}]
 dyn_braidouter_G_02_l [{shapeType=Capsule, length=3.8582, radius=2.3987, center=<3.5054 0 0>, rotation=<rot -90 0 0>}]
 dyn_braidouter_G_02_r [{shapeType=Capsule, length=3.8582, radius=2.3987, center=<3.5054 0 0>, rotation=<rot -90 0 0>}]
 dyn_braidouter_G_03_l [{shapeType=Capsule, length=6.2695, radius=1.5114, center=<4.2646 0 0>, rotation=<rot -90 0 0>}]
 dyn_braidouter_G_03_r [{shapeType=Capsule, length=6.2695, radius=1.5114, center=<4.2646 0 0>, rotation=<rot -90 0 0>}]
 dyn_braidouter_mid_00 [{
   shapeType=Capsule
   length=3.3499
   radius=1.9647
   shapeName=dyn_braidouter_mid_00_capsule
   center=<4.2909 -0.0987 0.8244>
   rotation=<rot 0 90.0037 -87.6043>
  }]
 dyn_braidouter_mid_00_MSTR [{shapeType=Capsule, length=0.505, radius=0.505, shapeName=dyn_braidouter_mid_00_MSTR_capsule, rotation=<rot 0 -90 0>}]
 dyn_braidouter_mid_01 [{shapeType=Capsule, length=5.5915, radius=2.3594, shapeName=dyn_braidouter_mid_01_capsule, center=<5 0 0>, rotation=<rot 90 0 0>}]
 dyn_braidouter_mid_02 [{shapeType=Capsule, length=5.9426, radius=2.3222, shapeName=dyn_braidouter_mid_02_capsule, center=<5.5 0 0>, rotation=<rot 90 0 0>}]
 dyn_braidouter_mid_03 [{shapeType=Capsule, length=6.8629, radius=1.802, shapeName=dyn_braidouter_mid_03_capsule, center=<5.5 0 0>, rotation=<rot 90 0 0>}]
 head [{shapeType=Capsule, length=11.9039, radius=9.4997, rotation=<rot 90 0.0264 0.0264>}]
 neck_02 [{shapeType=Capsule, length=12.6539, radius=8.7497, center=<5.798 -4.5539 -0.0001>, rotation=<rot 90 -0.0132 -0.0132>}]
 spine_05 [
  {shapeType=Capsule, length=27.4128, radius=9.6564, center=<4.6301 -2.624 0>}
  {shapeType=Capsule, length=10.0553, radius=7.4979, center=<8.8415 -3.5847 -19.8597>}
  {shapeType=Capsule, length=10.0553, radius=7.4979, center=<5.3888 -1.1883 19.0064>}
  {shapeType=Capsule, length=18.8717, radius=11.1284, center=<0.5385 -4.0682 -0.0001>}
 ]
 upperarm_l [
  {shapeType=Capsule, length=17.0232, radius=7.8901, center=<-3.2566 0.2521 4.3722>}
  {shapeType=Capsule, length=26, radius=4, center=<13.8743 0 0>, rotation=<rot 90 179.9604 179.9604>}
 ]
 upperarm_r [
  {shapeType=Capsule, length=15.123, radius=8.628, center=<0.8522 0.2515 -4.5404>, rotation=<rot -10.11 0 0>}
  {shapeType=Capsule, length=26.25, radius=3.75, center=<-14.325 0.0001 0>, rotation=<rot -90 0.0049 -0.0049>}
 ]
}
constraints [
 {bone1Name=dyn_braidouter_G_00_r, bone2Name=dyn_braidouter_G_00_r_MSTR, swing1LimitDegrees=5, swing2LimitDegrees=5, twistLimitDegrees=5}
 {bone1Name=dyn_braidouter_G_01_r, bone2Name=dyn_braidouter_G_00_r, swing1LimitDegrees=45, swing2LimitDegrees=45, twistLimitDegrees=5}
 {bone1Name=dyn_braidouter_G_02_r, bone2Name=dyn_braidouter_G_01_r, swing1LimitDegrees=20, swing2LimitDegrees=20, twistLimitDegrees=5}
 {bone1Name=dyn_braidouter_G_03_r, bone2Name=dyn_braidouter_G_02_r, swing1LimitDegrees=20, swing2LimitDegrees=20, twistLimitDegrees=5}
 {bone1Name=dyn_braidouter_F_00_r, bone2Name=dyn_braidouter_F_00_r_MSTR, swing1LimitDegrees=5, swing2LimitDegrees=5, twistLimitDegrees=5}
 {bone1Name=dyn_braidouter_F_01_r, bone2Name=dyn_braidouter_F_00_r, swing1LimitDegrees=5, swing2LimitDegrees=20, twistLimitDegrees=5}
 {bone1Name=dyn_braidouter_F_02_r, bone2Name=dyn_braidouter_F_01_r, swing1LimitDegrees=5, swing2LimitDegrees=20, twistLimitDegrees=5}
 {bone1Name=dyn_braidouter_F_03_r, bone2Name=dyn_braidouter_F_02_r, swing1LimitDegrees=20, swing2LimitDegrees=20, twistLimitDegrees=5}
 {bone1Name=dyn_braidouter_E_00_r, bone2Name=dyn_braidouter_E_00_r_MSTR, swing1LimitDegrees=25, swing2LimitDegrees=25, twistLimitDegrees=5}
 {bone1Name=dyn_braidouter_E_01_r, bone2Name=dyn_braidouter_E_00_r, swing1LimitDegrees=20, swing2LimitDegrees=5, twistLimitDegrees=5}
 {bone1Name=dyn_braidouter_E_02_r, bone2Name=dyn_braidouter_E_01_r, swing1LimitDegrees=20, swing2LimitDegrees=5, twistLimitDegrees=5}
 {bone1Name=dyn_braidouter_E_03_r, bone2Name=dyn_braidouter_E_02_r, swing1LimitDegrees=45, swing2LimitDegrees=45, twistLimitDegrees=5}
 {bone1Name=dyn_braidouter_D_00_r, bone2Name=dyn_braidouter_D_00_r_MSTR, swing1LimitDegrees=5, swing2LimitDegrees=5, twistLimitDegrees=5}
 {bone1Name=dyn_braidouter_D_01_r, bone2Name=dyn_braidouter_D_00_r, swing1LimitDegrees=35, swing2LimitDegrees=35, twistLimitDegrees=5}
 {bone1Name=dyn_braidouter_D_02_r, bone2Name=dyn_braidouter_D_01_r, swing1LimitDegrees=35, swing2LimitDegrees=35, twistLimitDegrees=5}
 {bone1Name=dyn_braidouter_D_03_r, bone2Name=dyn_braidouter_D_02_r, swing1LimitDegrees=35, swing2LimitDegrees=35, twistLimitDegrees=5}
 {bone1Name=dyn_braidouter_G_00_l, bone2Name=dyn_braidouter_G_00_l_MSTR, swing1LimitDegrees=5, swing2LimitDegrees=5, twistLimitDegrees=5}
 {bone1Name=dyn_braidouter_G_01_l, bone2Name=dyn_braidouter_G_00_l, swing1LimitDegrees=45, swing2LimitDegrees=45, twistLimitDegrees=5}
 {bone1Name=dyn_braidouter_G_02_l, bone2Name=dyn_braidouter_G_01_l, swing1LimitDegrees=20, swing2LimitDegrees=20, twistLimitDegrees=5}
 {bone1Name=dyn_braidouter_G_03_l, bone2Name=dyn_braidouter_G_02_l, swing1LimitDegrees=20, swing2LimitDegrees=20, twistLimitDegrees=5}
 {bone1Name=dyn_braidouter_F_00_l, bone2Name=dyn_braidouter_F_00_l_MSTR, swing1LimitDegrees=5, swing2LimitDegrees=5, twistLimitDegrees=5}
 {bone1Name=dyn_braidouter_F_01_l, bone2Name=dyn_braidouter_F_00_l, swing1LimitDegrees=5, swing2LimitDegrees=20, twistLimitDegrees=5}
 {bone1Name=dyn_braidouter_F_02_l, bone2Name=dyn_braidouter_F_01_l, swing1LimitDegrees=5, swing2LimitDegrees=20, twistLimitDegrees=5}
 {bone1Name=dyn_braidouter_F_03_l, bone2Name=dyn_braidouter_F_02_l, swing1LimitDegrees=20, swing2LimitDegrees=20, twistLimitDegrees=5}
 {bone1Name=dyn_braidouter_E_00_l, bone2Name=dyn_braidouter_E_00_l_MSTR, swing1LimitDegrees=5, swing2LimitDegrees=5, twistLimitDegrees=5}
 {bone1Name=dyn_braidouter_E_01_l, bone2Name=dyn_braidouter_E_00_l, swing1LimitDegrees=20, swing2LimitDegrees=5, twistLimitDegrees=5}
 {bone1Name=dyn_braidouter_E_02_l, bone2Name=dyn_braidouter_E_01_l, swing1LimitDegrees=20, swing2LimitDegrees=5, twistLimitDegrees=5}
 {bone1Name=dyn_braidouter_E_03_l, bone2Name=dyn_braidouter_E_02_l, swing1LimitDegrees=20, swing2LimitDegrees=20, twistLimitDegrees=5}
 {bone1Name=dyn_braidouter_D_00_l, bone2Name=dyn_braidouter_D_00_l_MSTR, swing1LimitDegrees=5, swing2LimitDegrees=5, twistLimitDegrees=5}
 {bone1Name=dyn_braidouter_D_01_l, bone2Name=dyn_braidouter_D_00_l, swing1LimitDegrees=20, swing2LimitDegrees=20, twistLimitDegrees=5}
 {bone1Name=dyn_braidouter_D_02_l, bone2Name=dyn_braidouter_D_01_l, swing1LimitDegrees=20, swing2LimitDegrees=20, twistLimitDegrees=5}
 {bone1Name=dyn_braidouter_D_03_l, bone2Name=dyn_braidouter_D_02_l, swing1LimitDegrees=20, swing2LimitDegrees=20, twistLimitDegrees=5}
 {
  bone1Name=upperarm_l
  bone2Name=spine_05
  swing1LimitDegrees=45
  swing1Motion=Free
  swing2LimitDegrees=45
  swing2Motion=Free
  twistLimitDegrees=45
  twistMotion=Free
 }
 {
  bone1Name=upperarm_r
  bone2Name=spine_05
  swing1LimitDegrees=45
  swing1Motion=Free
  swing2LimitDegrees=45
  swing2Motion=Free
  twistLimitDegrees=45
  twistMotion=Free
 }
 {
  bone1Name=head
  bone2Name=neck_02
  swing1LimitDegrees=45
  swing1Motion=Free
  swing2LimitDegrees=45
  swing2Motion=Free
  twistLimitDegrees=45
  twistMotion=Free
 }
 {
  bone1Name=dyn_braidouter_G_00_r_MSTR
  bone2Name=neck_02
  swing1LimitDegrees=45
  swing1Motion=Free
  swing2LimitDegrees=45
  swing2Motion=Free
  twistLimitDegrees=45
  twistMotion=Free
 }
 {
  bone1Name=dyn_braidouter_F_00_r_MSTR
  bone2Name=neck_02
  swing1LimitDegrees=45
  swing1Motion=Free
  swing2LimitDegrees=45
  swing2Motion=Free
  twistLimitDegrees=45
  twistMotion=Free
 }
 {
  bone1Name=dyn_braidouter_E_00_r_MSTR
  bone2Name=neck_02
  swing1LimitDegrees=45
  swing1Motion=Free
  swing2LimitDegrees=45
  swing2Motion=Free
  twistLimitDegrees=45
  twistMotion=Free
 }
 {
  bone1Name=dyn_braidouter_D_00_r_MSTR
  bone2Name=neck_02
  swing1LimitDegrees=45
  swing1Motion=Free
  swing2LimitDegrees=45
  swing2Motion=Free
  twistLimitDegrees=45
  twistMotion=Free
 }
 {
  bone1Name=dyn_braidouter_G_00_l_MSTR
  bone2Name=neck_02
  swing1LimitDegrees=45
  swing1Motion=Free
  swing2LimitDegrees=45
  swing2Motion=Free
  twistLimitDegrees=45
  twistMotion=Free
 }
 {
  bone1Name=dyn_braidouter_F_00_l_MSTR
  bone2Name=neck_02
  swing1LimitDegrees=45
  swing1Motion=Free
  swing2LimitDegrees=45
  swing2Motion=Free
  twistLimitDegrees=45
  twistMotion=Free
 }
 {
  bone1Name=dyn_braidouter_E_00_l_MSTR
  bone2Name=neck_02
  swing1LimitDegrees=45
  swing1Motion=Free
  swing2LimitDegrees=45
  swing2Motion=Free
  twistLimitDegrees=45
  twistMotion=Free
 }
 {
  bone1Name=dyn_braidouter_D_00_l_MSTR
  bone2Name=neck_02
  swing1LimitDegrees=45
  swing1Motion=Free
  swing2LimitDegrees=45
  swing2Motion=Free
  twistLimitDegrees=45
  twistMotion=Free
 }
 {
  bone1Name=neck_02
  bone2Name=spine_05
  swing1LimitDegrees=45
  swing1Motion=Free
  swing2LimitDegrees=45
  swing2Motion=Free
  twistLimitDegrees=45
  twistMotion=Free
 }
 {
  bone1Name=dyn_braidouter_mid_00_MSTR
  bone2Name=head
  swing1LimitDegrees=45
  swing1Motion=Free
  swing2LimitDegrees=45
  swing2Motion=Free
  twistLimitDegrees=45
  twistMotion=Free
 }
 {
  bone1Name=dyn_braidouter_mid_00
  bone2Name=dyn_braidouter_mid_00_MSTR
  swing1LimitDegrees=45
  swing1Motion=Free
  swing2LimitDegrees=45
  swing2Motion=Free
  twistLimitDegrees=45
  twistMotion=Free
 }
 {bone1Name=dyn_braidouter_mid_01, bone2Name=dyn_braidouter_mid_00, swing1LimitDegrees=25, swing2LimitDegrees=25, twistLimitDegrees=5}
 {bone1Name=dyn_braidouter_mid_02, bone2Name=dyn_braidouter_mid_01, swing1LimitDegrees=25, swing2LimitDegrees=25, twistLimitDegrees=5}
 {bone1Name=dyn_braidouter_mid_03, bone2Name=dyn_braidouter_mid_02, swing1LimitDegrees=25, swing2LimitDegrees=25, twistLimitDegrees=5}
]
