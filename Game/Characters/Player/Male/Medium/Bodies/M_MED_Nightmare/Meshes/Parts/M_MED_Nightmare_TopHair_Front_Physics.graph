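# Physics asset for the front top hair part of the male medium Nightmare player body, defining four braid chains (B and C, left and right).
# Each chain runs from a tiny master capsule through four capsules, with tight angular limits at the root and looser swing down the chain.
# Masters attach to the neck with free rotation, and free-rotation constraints cross-link each side's B and C chains at the third segment.
# Head, neck, upper spine and upper arms supply body capsules.
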PhysicsAsset /Game/Characters/Player/Male/Medium/Bodies/M_MED_Nightmare/Meshes/Parts/M_MED_Nightmare_TopHair_Front_Physics
bodies {
 dyn_braidouter_B_00_l [{shapeType=Capsule, length=4.9272, radius=1.734, center=<3.5036 0 0.0001>, rotation=<rot -90 1.2659 -1.2658>}]
 dyn_braidouter_B_00_l_MSTR [{shapeType=Capsule, length=0.01, radius=0.01}]
 dyn_braidouter_B_00_r [{shapeType=Capsule, length=4.9272, radius=1.734, center=<3.5036 0 0.0001>, rotation=<rot -90 1.6876 -1.6877>}]
 dyn_braidouter_B_00_r_MSTR [{shapeType=Capsule, length=0.01, radius=0.01}]
 dyn_braidouter_B_01_l [{shapeType=Capsule, length=4.9272, radius=1.734, center=<3.5036 0 0.0001>, rotation=<rot -90 -1.2655 1.2652>}]
 dyn_braidouter_B_01_r [{shapeType=Capsule, length=4.9272, radius=1.734, center=<3.5036 0 0.0001>, rotation=<rot -90 10.025 -10.0251>}]
 dyn_braidouter_B_02_l [{shapeType=Capsule, length=4.9272, radius=1.734, center=<3.5036 0 0.0001>, rotation=<rot -90 -144.7357 144.7354>}]
 dyn_braidouter_B_02_r [{shapeType=Capsule, length=4.9272, radius=1.734, center=<3.5036 0 0.0001>, rotation=<rot -90 177.4695 -177.4696>}]
 dyn_braidouter_B_03_l [{shapeType=Capsule, length=4.9272, radius=1.734, center=<3.5036 0 0.0001>, rotation=<rot -90 -1.2657 1.2654>}]
 dyn_braidouter_B_03_r [{shapeType=Capsule, length=4.9272, radius=1.734, center=<3.5036 0 0.0001>, rotation=<rot -90 -4.4924 4.492>}]
 dyn_braidouter_C_00_l [{shapeType=Capsule, length=3.6324, radius=3.0288, center=<3.5036 0 0.0001>, rotation=<rot -90 35.2644 -35.2643>}]
 dyn_braidouter_C_00_l_MSTR [{shapeType=Capsule, length=0.01, radius=0.01}]
 dyn_braidouter_C_00_r [{shapeType=Capsule, length=3.6324, radius=3.0288, center=<3.5036 0 0.0001>, rotation=<rot -90 35.2645 -35.2646>}]
 dyn_braidouter_C_00_r_MSTR [{shapeType=Capsule, length=0.01, radius=0.01}]
 dyn_braidouter_C_01_l [{shapeType=Capsule, length=5.6565, radius=2.3987, center=<3.2354 -0.0001 0>, rotation=<rot -90 -0.844 0.8438>}]
 dyn_braidouter_C_01_r [{shapeType=Capsule, length=5.6565, radius=2.3987, center=<3.2354 -0.0001 0>, rotation=<rot -90 1.6876 -1.6877>}]
 dyn_braidouter_C_02_l [{shapeType=Capsule, length=3.8582, radius=2.3987, center=<3.5054 0 -0.0001>, rotation=<rot -90 0 -0.0001>}]
 dyn_braidouter_C_02_r [{shapeType=Capsule, length=3.8582, radius=2.3987, center=<3.5054 0 -0.0001>, rotation=<rot -90 35.2643 -35.2642>}]
 dyn_braidouter_C_03_l [{shapeType=Capsule, length=6.2533, radius=1.5276, center=<4.2646 0 0.0002>, rotation=<rot -90 35.2645 -35.2646>}]
 dyn_braidouter_C_03_r [{shapeType=Capsule, length=5.7424, radius=2.0386, center=<4.2646 0 0.0002>, rotation=<rot -90 14.851 -14.851>}]
 head [{shapeType=Capsule, length=11.9039, radius=9.4997, rotation=<rot 90 0.0264 0.0264>}]
 neck_02 [{shapeType=Capsule, length=12.6539, radius=8.7497, center=<5.798 -4.5539 -0.0001>, rotation=<rot 90 -0.0132 -0.0132>}]
 spine_05 [
  {shapeType=Capsule, length=27.4128, radius=9.6564, center=<4.6301 -2.624 0>}
  {shapeType=Capsule, length=10.0553, radius=7.4979, center=<8.8415 -3.5847 -19.8597>}
  {shapeType=Capsule, length=10.0553, radius=7.4979, center=<5.3888 -1.1883 19.0064>}
  {shapeType=Capsule, length=18.6697, radius=11.3303, center=<0.5385 -4.0682 -0.0001>}
 ]
 upperarm_l [
  {shapeType=Capsule, length=17.0232, radius=7.8901, center=<-3.2566 0.2521 4.4523>}
  {shapeType=Capsule, length=26, radius=4, center=<13.8743 0 0>, rotation=<rot 90 179.9604 179.9604>}
 ]
 upperarm_r [
  {shapeType=Capsule, length=10.398, radius=7.5558, center=<1.2568 -0.0002 -1.9793>, rotation=<rot -9.9973 0 180>}
  {shapeType=Capsule, length=26.25, radius=3.75, center=<-14.325 0.0001 0>, rotation=<rot -90 0.0049 -0.0049>}
 ]
}
constraints [
 {bone1Name=dyn_braidouter_C_00_r, bone2Name=dyn_braidouter_C_00_r_MSTR, swing1LimitDegrees=5, swing2LimitDegrees=5, twistLimitDegrees=5}
 {bone1Name=dyn_braidouter_C_01_r, bone2Name=dyn_braidouter_C_00_r, swing1LimitDegrees=20, swing2LimitDegrees=20, twistLimitDegrees=5}
 {bone1Name=dyn_braidouter_C_02_r, bone2Name=dyn_braidouter_C_01_r, swing1LimitDegrees=20, swing2LimitDegrees=20, twistLimitDegrees=5}
 {bone1Name=dyn_braidouter_C_03_r, bone2Name=dyn_braidouter_C_02_r, swing1LimitDegrees=20, swing2LimitDegrees=20, twistLimitDegrees=5}
 {bone1Name=dyn_braidouter_B_00_l, bone2Name=dyn_braidouter_B_00_l_MSTR, swing1LimitDegrees=5, swing2LimitDegrees=5, twistLimitDegrees=5}
 {bone1Name=dyn_braidouter_B_01_l, bone2Name=dyn_braidouter_B_00_l, swing1LimitDegrees=20, swing2LimitDegrees=20, twistLimitDegrees=5}
 {bone1Name=dyn_braidouter_B_02_l, bone2Name=dyn_braidouter_B_01_l, swing1LimitDegrees=20, swing2LimitDegrees=20, twistLimitDegrees=5}
 {bone1Name=dyn_braidouter_B_03_l, bone2Name=dyn_braidouter_B_02_l, swing1LimitDegrees=20, swing2LimitDegrees=20, twistLimitDegrees=5}
 {bone1Name=dyn_braidouter_C_00_l, bone2Name=dyn_braidouter_C_00_l_MSTR, swing1LimitDegrees=5, swing2LimitDegrees=5, twistLimitDegrees=5}
 {bone1Name=dyn_braidouter_C_01_l, bone2Name=dyn_braidouter_C_00_l, swing1LimitDegrees=20, swing2LimitDegrees=20, twistLimitDegrees=5}
 {bone1Name=dyn_braidouter_C_02_l, bone2Name=dyn_braidouter_C_01_l, swing1LimitDegrees=20, swing2LimitDegrees=20, twistLimitDegrees=5}
 {bone1Name=dyn_braidouter_C_03_l, bone2Name=dyn_braidouter_C_02_l, swing1LimitDegrees=20, swing2LimitDegrees=20, twistLimitDegrees=5}
 {bone1Name=dyn_braidouter_B_00_r, bone2Name=dyn_braidouter_B_00_r_MSTR, swing1LimitDegrees=5, swing2LimitDegrees=5, twistLimitDegrees=5}
 {bone1Name=dyn_braidouter_B_01_r, bone2Name=dyn_braidouter_B_00_r, swing1LimitDegrees=20, swing2LimitDegrees=20, twistLimitDegrees=5}
 {bone1Name=dyn_braidouter_B_02_r, bone2Name=dyn_braidouter_B_01_r, swing1LimitDegrees=20, swing2LimitDegrees=20, twistLimitDegrees=5}
 {bone1Name=dyn_braidouter_B_03_r, bone2Name=dyn_braidouter_B_02_r, swing1LimitDegrees=20, swing2LimitDegrees=20, twistLimitDegrees=5}
 {
  bone1Name=upperarm_l
  bone2Name=spine_05
  swing1LimitDegrees=45
  swing1Motion=Free
  swing2LimitDegrees=45
  swing2Motion=Free
  twistLimitDegrees=45
  twistMotion=Free
 }
 {
  bone1Name=upperarm_r
  bone2Name=spine_05
  swing1LimitDegrees=45
  swing1Motion=Free
  swing2LimitDegrees=45
  swing2Motion=Free
  twistLimitDegrees=45
  twistMotion=Free
 }
 {
  bone1Name=head
  bone2Name=neck_02
  swing1LimitDegrees=45
  swing1Motion=Free
  swing2LimitDegrees=45
  swing2Motion=Free
  twistLimitDegrees=45
  twistMotion=Free
 }
 {
  bone1Name=dyn_braidouter_C_00_r_MSTR
  bone2Name=neck_02
  swing1LimitDegrees=45
  swing1Motion=Free
  swing2LimitDegrees=45
  swing2Motion=Free
  twistLimitDegrees=45
  twistMotion=Free
 }
 {
  bone1Name=dyn_braidouter_B_00_r_MSTR
  bone2Name=neck_02
  swing1LimitDegrees=45
  swing1Motion=Free
  swing2LimitDegrees=45
  swing2Motion=Free
  twistLimitDegrees=45
  twistMotion=Free
 }
 {
  bone1Name=dyn_braidouter_B_00_l_MSTR
  bone2Name=neck_02
  swing1LimitDegrees=45
  swing1Motion=Free
  swing2LimitDegrees=45
  swing2Motion=Free
  twistLimitDegrees=45
  twistMotion=Free
 }
 {
  bone1Name=dyn_braidouter_C_00_l_MSTR
  bone2Name=neck_02
  swing1LimitDegrees=45
  swing1Motion=Free
  swing2LimitDegrees=45
  swing2Motion=Free
  twistLimitDegrees=45
  twistMotion=Free
 }
 {
  bone1Name=neck_02
  bone2Name=spine_05
  swing1LimitDegrees=45
  swing1Motion=Free
  swing2LimitDegrees=45
  swing2Motion=Free
  twistLimitDegrees=45
  twistMotion=Free
 }
 {
  bone1Name=dyn_braidouter_C_02_r
  bone2Name=dyn_braidouter_B_02_r
  swing1LimitDegrees=45
  swing1Motion=Free
  swing2LimitDegrees=45
  swing2Motion=Free
  twistLimitDegrees=45
  twistMotion=Free
 }
 {
  bone1Name=dyn_braidouter_C_02_l
  bone2Name=dyn_braidouter_B_02_l
  swing1LimitDegrees=45
  swing1Motion=Free
  swing2LimitDegrees=45
  swing2Motion=Free
  twistLimitDegrees=45
  twistMotion=Free
 }
]
