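# Post-process style AnimBlueprint for the M_MED_Nightmare character part, adding hair physics to the pose copied from the body mesh.
# RigidBody nodes simulate top and bottom hair, then chained AnimDynamics nodes spring six braids against shoulder spheres; the front-hair node is unconnected.
# Short backbling offsets a backpack collision bone, and skydiving force replaces top-hair gravity.
# While targeting, head-to-neck rotation blends in a lazer_01 bone rotation.
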AnimBlueprint /BRCosmetics/Characters/Player/Male/Medium/Bodies/M_MED_Nightmare/Meshes/Parts/M_MED_Nightmare_AnimBP
target_skeleton=/Game/Characters/Player/Male/Medium/Bodies/M_MED_Nightmare/Meshes/Parts/M_MED_Nightmare_Skeleton
graphs [
 AnimGraph {nodes [
   RigidBody!(ComponentPose, Alpha, OverrideWorldGravity:Vector="0,0,-980", ExternalForce) -> (Pose) {
    bClampLinearTranslationLimitToRefPose=true
    bOverrideWorldGravity=true
    overridePhysicsAsset=/Game/Characters/Player/Male/Medium/Bodies/M_MED_Nightmare/Meshes/Parts/M_MED_Nightmare_TopHair_Front_Physics
    simulationSpace=BaseBoneSpace
    alphaBoolBlend {blendOption=Linear}
    baseBoneRef {boneName=neck_02}
    componentAppliedLinearAccClamp=<1500 1500 1500>
    componentLinearAccScale=<0.45 0.45 0.45>
    componentLinearVelScale=<0.45 0.45 0.45>
    overrideWorldGravity {z=-980}
   }
   n1 = Get Body_Mesh()
   n2 = "Transform (Modify) Bone - Bone: dyn_backpack_COL"(
    ComponentPose=Local To Component(LocalPose=Copy Pose From Mesh(SourceMeshComponent=@n1) {bUseAttachedParent=true})
    Alpha=Get isWearingBackbling() -> (isWearingBackbling:double)
    Translation="63.872,0.000,0.626"
    Rotation="0,0,0"
    Scale="1,1,1"
   ) {
    rotationMode=BMM_Additive
    rotationSpace=BCS_BoneSpace
    scaleMode=BMM_Additive
    scaleSpace=BCS_BoneSpace
    translationMode=BMM_Additive
    translationSpace=BCS_BoneSpace
    alphaBoolBlend {blendOption=Linear}
    boneToModify {boneName=dyn_backpack_COL}
    translation {x=63.872, z=0.626}
   }
   n3 = RigidBody(ComponentPose=@n2, OverrideWorldGravity=Get windVector() -> (windVector:Vector)) {
    bClampLinearTranslationLimitToRefPose=true
    bOverrideWorldGravity=true
    overridePhysicsAsset=&/M_MED_Nightmare_TopHair_Physics
    simulationSpace=BaseBoneSpace
    alphaBoolBlend {blendOption=Linear}
    baseBoneRef {boneName=neck_02}
    componentAppliedLinearAccClamp=<1500 1500 1500>
    componentLinearAccScale=<0.35 0.35 0.35>
    componentLinearVelScale=<0.35 0.35 0.35>
    overrideWorldGravity {z=-980}
   }
   n4 = RigidBody(ComponentPose=@n3, OverrideWorldGravity="0,0,-980") {
    bClampLinearTranslationLimitToRefPose=true
    bOverrideWorldGravity=true
    overridePhysicsAsset=&/M_MED_Nightmare_BottomHair_Physics
    simulationSpace=BaseBoneSpace
    alphaBoolBlend {blendOption=Linear}
    baseBoneRef {boneName=neck_02}
    componentAppliedLinearAccClamp=<1500 1500 1500>
    componentLinearAccScale=<0.3 0.3 0.3>
    componentLinearVelScale=<0.3 0.3 0.3>
    overrideWorldGravity {z=-980}
   }
   n5 = Anim Dynamics - Bone: dyn_braidouter_C_00_r(ComponentPose=@n4, Alpha=1, ExternalForce="0,0,0") {
    angularDampingOverride=0.9999
    angularSpringConstant=75
    bAngularSpring=true
    bGravityOverrideInSimSpace=true
    bOverrideAngularDamping=true
    bUseSphericalLimits=true
    simulationSpace=RootRelative
    alphaBoolBlend {blendOption=Linear}
    boundBone {boneName=dyn_braidouter_C_00_r}
    componentAppliedLinearAccClamp=<3500 3500 3500>
    componentLinearAccScale=<0.9 0.9 0.9>
    componentLinearVelScale=<0.9 0.9 0.9>
    physicsBodyDefinitions [{
      collisionType=CustomSphere
      sphereCollisionRadius=4
      boundBone {boneName=dyn_braidouter_C_00_r}
      boxExtents=<25 15 15>
      constraintSetup {angularLimitsMax=<5 0 25>, angularLimitsMin=<-5 -35 -25>, angularTarget=<0 0 0>}
      localJointOffset=<2 0 0>
     }]
    relativeSpaceBone {boneName=root}
    sphericalLimits [{limitRadius=6, drivingBone {boneName=upperarm_twist_02_l}, sphereLocalOffset=<0 0 -3>}]
   }
   n6 = Anim Dynamics - Bone: dyn_braidouter_C_01_r(ComponentPose=@n5, Alpha=1, ExternalForce="0,0,0") {
    angularDampingOverride=0.9999
    angularSpringConstant=75
    bAngularSpring=true
    bGravityOverrideInSimSpace=true
    bOverrideAngularDamping=true
    bUseSphericalLimits=true
    simulationSpace=RootRelative
    alphaBoolBlend {blendOption=Linear}
    boundBone {boneName=dyn_braidouter_C_01_r}
    componentAppliedLinearAccClamp=<3500 3500 3500>
    componentLinearAccScale=<0.9 0.9 0.9>
    componentLinearVelScale=<0.9 0.9 0.9>
    physicsBodyDefinitions [{
      collisionType=CustomSphere
      sphereCollisionRadius=4
      boundBone {boneName=dyn_braidouter_C_01_r}
      boxExtents=<25 15 15>
      constraintSetup {angularLimitsMax=<5 5 25>, angularLimitsMin=<-5 -35 -25>, angularTarget=<0 0 0>}
      localJointOffset=<2 0 0>
     }]
    relativeSpaceBone {boneName=root}
    sphericalLimits [{limitRadius=20, drivingBone {boneName=clavicle_r}, sphereLocalOffset=<-15 10 10>}]
   }
   n7 = Anim Dynamics - Bone: dyn_braidouter_C_02_r(ComponentPose=@n6, Alpha=1, ExternalForce="0,0,0") {
    angularDampingOverride=0.9999
    angularSpringConstant=75
    bAngularSpring=true
    bGravityOverrideInSimSpace=true
    bOverrideAngularDamping=true
    bUseSphericalLimits=true
    simulationSpace=RootRelative
    alphaBoolBlend {blendOption=Linear}
    boundBone {boneName=dyn_braidouter_C_02_r}
    componentAppliedLinearAccClamp=<3500 3500 3500>
    componentLinearAccScale=<0.9 0.9 0.9>
    componentLinearVelScale=<0.9 0.9 0.9>
    physicsBodyDefinitions [{
      collisionType=CustomSphere
      sphereCollisionRadius=4
      boundBone {boneName=dyn_braidouter_C_02_r}
      boxExtents=<25 15 15>
      constraintSetup {angularLimitsMax=<5 10 25>, angularLimitsMin=<-5 -35 -25>, angularTarget=<0 0 0>}
      localJointOffset=<2 0 0>
     }]
    relativeSpaceBone {boneName=root}
    sphericalLimits [{limitRadius=20, drivingBone {boneName=clavicle_r}, sphereLocalOffset=<-15 10 10>}]
   }
   n8 = Anim Dynamics - Bone: dyn_braidouter_C_03_r(ComponentPose=@n7, Alpha=1, ExternalForce="0,0,0") {
    angularDampingOverride=0.9999
    angularSpringConstant=75
    bAngularSpring=true
    bGravityOverrideInSimSpace=true
    bOverrideAngularDamping=true
    bUseSphericalLimits=true
    simulationSpace=RootRelative
    alphaBoolBlend {blendOption=Linear}
    boundBone {boneName=dyn_braidouter_C_03_r}
    componentAppliedLinearAccClamp=<3500 3500 3500>
    componentLinearAccScale=<0.9 0.9 0.9>
    componentLinearVelScale=<0.9 0.9 0.9>
    physicsBodyDefinitions [{
      collisionType=CustomSphere
      sphereCollisionRadius=4
      boundBone {boneName=dyn_braidouter_C_03_r}
      boxExtents=<25 15 15>
      constraintSetup {angularLimitsMax=<5 20 25>, angularLimitsMin=<-5 -35 -25>, angularTarget=<0 0 0>}
      localJointOffset=<2 0 0>
     }]
    relativeSpaceBone {boneName=root}
    sphericalLimits [{limitRadius=20, drivingBone {boneName=clavicle_r}, sphereLocalOffset=<-15 10 10>}]
   }
   n9 = Anim Dynamics - Bone: dyn_braidouter_C_04_r(ComponentPose=@n8, Alpha=1, ExternalForce="0,0,0") {
    angularDampingOverride=0.9999
    angularSpringConstant=75
    bAngularSpring=true
    bGravityOverrideInSimSpace=true
    bOverrideAngularDamping=true
    bUseSphericalLimits=true
    simulationSpace=RootRelative
    alphaBoolBlend {blendOption=Linear}
    boundBone {boneName=dyn_braidouter_C_04_r}
    componentAppliedLinearAccClamp=<3500 3500 3500>
    componentLinearAccScale=<0.9 0.9 0.9>
    componentLinearVelScale=<0.9 0.9 0.9>
    physicsBodyDefinitions [{
      collisionType=CustomSphere
      sphereCollisionRadius=4
      boundBone {boneName=dyn_braidouter_C_04_r}
      boxExtents=<25 15 15>
      constraintSetup {angularLimitsMax=<5 35 25>, angularLimitsMin=<-5 -35 -25>, angularTarget=<0 0 0>}
      localJointOffset=<2 0 0>
     }]
    relativeSpaceBone {boneName=root}
    sphericalLimits [{limitRadius=20, drivingBone {boneName=clavicle_r}, sphereLocalOffset=<-15 10 10>}]
   }
   n10 = Anim Dynamics - Bone: dyn_braidouter_A_00_r(ComponentPose=@n9, Alpha=1, ExternalForce="0,0,0") {
    angularDampingOverride=0.9999
    angularSpringConstant=75
    bAngularSpring=true
    bGravityOverrideInSimSpace=true
    bOverrideAngularDamping=true
    bUseSphericalLimits=true
    simulationSpace=RootRelative
    alphaBoolBlend {blendOption=Linear}
    boundBone {boneName=dyn_braidouter_A_00_r}
    componentAppliedLinearAccClamp=<3500 3500 3500>
    componentLinearAccScale=<0.9 0.9 0.9>
    componentLinearVelScale=<0.9 0.9 0.9>
    physicsBodyDefinitions [{
      collisionType=CustomSphere
      sphereCollisionRadius=4
      boundBone {boneName=dyn_braidouter_A_00_r}
      boxExtents=<25 15 15>
      constraintSetup {angularLimitsMax=<5 0 25>, angularLimitsMin=<-5 -35 -25>, angularTarget=<0 0 0>}
      localJointOffset=<2 0 0>
     }]
    relativeSpaceBone {boneName=root}
    sphericalLimits [{limitRadius=6, drivingBone {boneName=upperarm_twist_02_l}, sphereLocalOffset=<0 0 -3>}]
   }
   n11 = Anim Dynamics - Bone: dyn_braidouter_A_01_r(ComponentPose=@n10, Alpha=1, ExternalForce="0,0,0") {
    angularDampingOverride=0.9999
    angularSpringConstant=75
    bAngularSpring=true
    bGravityOverrideInSimSpace=true
    bOverrideAngularDamping=true
    bUseSphericalLimits=true
    simulationSpace=RootRelative
    alphaBoolBlend {blendOption=Linear}
    boundBone {boneName=dyn_braidouter_A_01_r}
    componentAppliedLinearAccClamp=<3500 3500 3500>
    componentLinearAccScale=<0.9 0.9 0.9>
    componentLinearVelScale=<0.9 0.9 0.9>
    physicsBodyDefinitions [{
      collisionType=CustomSphere
      sphereCollisionRadius=4
      boundBone {boneName=dyn_braidouter_A_01_r}
      boxExtents=<25 15 15>
      constraintSetup {angularLimitsMax=<5 5 25>, angularLimitsMin=<-5 -35 -25>, angularTarget=<0 0 0>}
      localJointOffset=<2 0 0>
     }]
    relativeSpaceBone {boneName=root}
    sphericalLimits [{limitRadius=20, drivingBone {boneName=upperarm_twist_02_l}, sphereLocalOffset=<-15 10 10>}]
   }
   n12 = Anim Dynamics - Bone: dyn_braidouter_A_02_r(ComponentPose=@n11, Alpha=1, ExternalForce="0,0,0") {
    angularDampingOverride=0.9999
    angularSpringConstant=75
    bAngularSpring=true
    bGravityOverrideInSimSpace=true
    bOverrideAngularDamping=true
    bUseSphericalLimits=true
    simulationSpace=RootRelative
    alphaBoolBlend {blendOption=Linear}
    boundBone {boneName=dyn_braidouter_A_02_r}
    componentAppliedLinearAccClamp=<3500 3500 3500>
    componentLinearAccScale=<0.9 0.9 0.9>
    componentLinearVelScale=<0.9 0.9 0.9>
    physicsBodyDefinitions [{
      collisionType=CustomSphere
      sphereCollisionRadius=4
      boundBone {boneName=dyn_braidouter_A_02_r}
      boxExtents=<25 15 15>
      constraintSetup {angularLimitsMax=<5 10 25>, angularLimitsMin=<-5 -35 -25>, angularTarget=<0 0 0>}
      localJointOffset=<2 0 0>
     }]
    relativeSpaceBone {boneName=root}
    sphericalLimits [{limitRadius=20, drivingBone {boneName=clavicle_r}, sphereLocalOffset=<-15 10 10>}]
   }
   n13 = Anim Dynamics - Bone: dyn_braidouter_A_03_r(ComponentPose=@n12, Alpha=1, ExternalForce="0,0,0") {
    angularDampingOverride=0.9999
    angularSpringConstant=75
    bAngularSpring=true
    bGravityOverrideInSimSpace=true
    bOverrideAngularDamping=true
    bUseSphericalLimits=true
    simulationSpace=RootRelative
    alphaBoolBlend {blendOption=Linear}
    boundBone {boneName=dyn_braidouter_A_03_r}
    componentAppliedLinearAccClamp=<3500 3500 3500>
    componentLinearAccScale=<0.9 0.9 0.9>
    componentLinearVelScale=<0.9 0.9 0.9>
    physicsBodyDefinitions [{
      collisionType=CustomSphere
      sphereCollisionRadius=4
      boundBone {boneName=dyn_braidouter_A_03_r}
      boxExtents=<25 15 15>
      constraintSetup {angularLimitsMax=<5 20 25>, angularLimitsMin=<-5 -35 -25>, angularTarget=<0 0 0>}
      localJointOffset=<2 0 0>
     }]
    relativeSpaceBone {boneName=root}
    sphericalLimits [{limitRadius=20, drivingBone {boneName=clavicle_r}, sphereLocalOffset=<-15 10 10>}]
   }
   n14 = Anim Dynamics - Bone: dyn_braidouter_A_04_r(ComponentPose=@n13, Alpha=1, ExternalForce="0,0,0") {
    angularDampingOverride=0.9999
    angularSpringConstant=75
    bAngularSpring=true
    bGravityOverrideInSimSpace=true
    bOverrideAngularDamping=true
    bUseSphericalLimits=true
    simulationSpace=RootRelative
    alphaBoolBlend {blendOption=Linear}
    boundBone {boneName=dyn_braidouter_A_04_r}
    componentAppliedLinearAccClamp=<3500 3500 3500>
    componentLinearAccScale=<0.9 0.9 0.9>
    componentLinearVelScale=<0.9 0.9 0.9>
    physicsBodyDefinitions [{
      collisionType=CustomSphere
      sphereCollisionRadius=4
      boundBone {boneName=dyn_braidouter_A_04_r}
      boxExtents=<25 15 15>
      constraintSetup {angularLimitsMax=<5 35 25>, angularLimitsMin=<-5 -35 -25>, angularTarget=<0 0 0>}
      localJointOffset=<2 0 0>
     }]
    relativeSpaceBone {boneName=root}
    sphericalLimits [{limitRadius=20, drivingBone {boneName=clavicle_r}, sphereLocalOffset=<-15 10 10>}]
   }
   n15 = Anim Dynamics - Bone: dyn_braidouter_B_00_r(ComponentPose=@n14, Alpha=1, ExternalForce="0,0,0") {
    angularDampingOverride=0.9999
    angularSpringConstant=75
    bAngularSpring=true
    bGravityOverrideInSimSpace=true
    bOverrideAngularDamping=true
    bUseSphericalLimits=true
    simulationSpace=RootRelative
    alphaBoolBlend {blendOption=Linear}
    boundBone {boneName=dyn_braidouter_B_00_r}
    componentAppliedLinearAccClamp=<3500 3500 3500>
    componentLinearAccScale=<0.9 0.9 0.9>
    componentLinearVelScale=<0.9 0.9 0.9>
    physicsBodyDefinitions [{
      collisionType=CustomSphere
      sphereCollisionRadius=4
      boundBone {boneName=dyn_braidouter_B_00_r}
      boxExtents=<25 15 15>
      constraintSetup {angularLimitsMax=<5 0 25>, angularLimitsMin=<-5 -35 -25>, angularTarget=<0 0 0>}
      localJointOffset=<2 0 0>
     }]
    relativeSpaceBone {boneName=root}
    sphericalLimits [{limitRadius=6, drivingBone {boneName=upperarm_twist_02_l}, sphereLocalOffset=<0 0 -3>}]
   }
   n16 = Anim Dynamics - Bone: dyn_braidouter_B_01_r(ComponentPose=@n15, Alpha=1, ExternalForce="0,0,0") {
    angularDampingOverride=0.9999
    angularSpringConstant=75
    bAngularSpring=true
    bGravityOverrideInSimSpace=true
    bOverrideAngularDamping=true
    bUseSphericalLimits=true
    simulationSpace=RootRelative
    alphaBoolBlend {blendOption=Linear}
    boundBone {boneName=dyn_braidouter_B_01_r}
    componentAppliedLinearAccClamp=<3500 3500 3500>
    componentLinearAccScale=<0.9 0.9 0.9>
    componentLinearVelScale=<0.9 0.9 0.9>
    physicsBodyDefinitions [{
      collisionType=CustomSphere
      sphereCollisionRadius=4
      boundBone {boneName=dyn_braidouter_B_01_r}
      boxExtents=<25 15 15>
      constraintSetup {angularLimitsMax=<5 5 25>, angularLimitsMin=<-5 -35 -25>, angularTarget=<0 0 0>}
      localJointOffset=<2 0 0>
     }]
    relativeSpaceBone {boneName=root}
    sphericalLimits [{limitRadius=20, drivingBone {boneName=upperarm_twist_02_l}, sphereLocalOffset=<-15 10 10>}]
   }
   n17 = Anim Dynamics - Bone: dyn_braidouter_B_02_r(ComponentPose=@n16, Alpha=1, ExternalForce="0,0,0") {
    angularDampingOverride=0.9999
    angularSpringConstant=75
    bAngularSpring=true
    bGravityOverrideInSimSpace=true
    bOverrideAngularDamping=true
    bUseSphericalLimits=true
    simulationSpace=RootRelative
    alphaBoolBlend {blendOption=Linear}
    boundBone {boneName=dyn_braidouter_B_02_r}
    componentAppliedLinearAccClamp=<3500 3500 3500>
    componentLinearAccScale=<0.9 0.9 0.9>
    componentLinearVelScale=<0.9 0.9 0.9>
    physicsBodyDefinitions [{
      collisionType=CustomSphere
      sphereCollisionRadius=4
      boundBone {boneName=dyn_braidouter_B_02_r}
      boxExtents=<25 15 15>
      constraintSetup {angularLimitsMax=<5 10 25>, angularLimitsMin=<-5 -35 -25>, angularTarget=<0 0 0>}
      localJointOffset=<2 0 0>
     }]
    relativeSpaceBone {boneName=root}
    sphericalLimits [{limitRadius=20, drivingBone {boneName=clavicle_r}, sphereLocalOffset=<-15 10 10>}]
   }
   n18 = Anim Dynamics - Bone: dyn_braidouter_B_03_r(ComponentPose=@n17, Alpha=1, ExternalForce="0,0,0") {
    angularDampingOverride=0.9999
    angularSpringConstant=75
    bAngularSpring=true
    bGravityOverrideInSimSpace=true
    bOverrideAngularDamping=true
    bUseSphericalLimits=true
    simulationSpace=RootRelative
    alphaBoolBlend {blendOption=Linear}
    boundBone {boneName=dyn_braidouter_B_03_r}
    componentAppliedLinearAccClamp=<3500 3500 3500>
    componentLinearAccScale=<0.9 0.9 0.9>
    componentLinearVelScale=<0.9 0.9 0.9>
    physicsBodyDefinitions [{
      collisionType=CustomSphere
      sphereCollisionRadius=4
      boundBone {boneName=dyn_braidouter_B_03_r}
      boxExtents=<25 15 15>
      constraintSetup {angularLimitsMax=<5 20 25>, angularLimitsMin=<-5 -35 -25>, angularTarget=<0 0 0>}
      localJointOffset=<2 0 0>
     }]
    relativeSpaceBone {boneName=root}
    sphericalLimits [{limitRadius=20, drivingBone {boneName=clavicle_r}, sphereLocalOffset=<-15 10 10>}]
   }
   n19 = Anim Dynamics - Bone: dyn_braidouter_B_04_r(ComponentPose=@n18, Alpha=1, ExternalForce="0,0,0") {
    angularDampingOverride=0.9999
    angularSpringConstant=75
    bAngularSpring=true
    bGravityOverrideInSimSpace=true
    bOverrideAngularDamping=true
    bUseSphericalLimits=true
    simulationSpace=RootRelative
    alphaBoolBlend {blendOption=Linear}
    boundBone {boneName=dyn_braidouter_B_04_r}
    componentAppliedLinearAccClamp=<3500 3500 3500>
    componentLinearAccScale=<0.9 0.9 0.9>
    componentLinearVelScale=<0.9 0.9 0.9>
    physicsBodyDefinitions [{
      collisionType=CustomSphere
      sphereCollisionRadius=4
      boundBone {boneName=dyn_braidouter_B_04_r}
      boxExtents=<25 15 15>
      constraintSetup {angularLimitsMax=<5 35 25>, angularLimitsMin=<-5 -35 -25>, angularTarget=<0 0 0>}
      localJointOffset=<2 0 0>
     }]
    relativeSpaceBone {boneName=root}
    sphericalLimits [{limitRadius=20, drivingBone {boneName=clavicle_r}, sphereLocalOffset=<-15 10 10>}]
   }
   n20 = Anim Dynamics - Bone: dyn_braidouter_C_00_l(ComponentPose=@n19, Alpha=1, ExternalForce="0,0,0") {
    angularDampingOverride=0.9999
    angularSpringConstant=75
    bAngularSpring=true
    bGravityOverrideInSimSpace=true
    bOverrideAngularDamping=true
    bUseSphericalLimits=true
    simulationSpace=RootRelative
    alphaBoolBlend {blendOption=Linear}
    boundBone {boneName=dyn_braidouter_C_00_l}
    componentAppliedLinearAccClamp=<3500 3500 3500>
    componentLinearAccScale=<0.9 0.9 0.9>
    componentLinearVelScale=<0.9 0.9 0.9>
    physicsBodyDefinitions [{
      collisionType=CustomSphere
      sphereCollisionRadius=4
      boundBone {boneName=dyn_braidouter_C_00_l}
      boxExtents=<25 15 15>
      constraintSetup {angularLimitsMax=<5 0 25>, angularLimitsMin=<-5 -35 -25>, angularTarget=<0 0 0>}
      localJointOffset=<2 0 0>
     }]
    relativeSpaceBone {boneName=root}
    sphericalLimits [{limitRadius=6, drivingBone {boneName=clavicle_l}, sphereLocalOffset=<0 0 -3>}]
   }
   n21 = Anim Dynamics - Bone: dyn_braidouter_C_01_l(ComponentPose=@n20, Alpha=1, ExternalForce="0,0,0") {
    angularDampingOverride=0.9999
    angularSpringConstant=75
    bAngularSpring=true
    bGravityOverrideInSimSpace=true
    bOverrideAngularDamping=true
    bUseSphericalLimits=true
    simulationSpace=RootRelative
    alphaBoolBlend {blendOption=Linear}
    boundBone {boneName=dyn_braidouter_C_01_l}
    componentAppliedLinearAccClamp=<3500 3500 3500>
    componentLinearAccScale=<0.9 0.9 0.9>
    componentLinearVelScale=<0.9 0.9 0.9>
    physicsBodyDefinitions [{
      collisionType=CustomSphere
      sphereCollisionRadius=4
      boundBone {boneName=dyn_braidouter_C_01_l}
      boxExtents=<25 15 15>
      constraintSetup {angularLimitsMax=<5 5 25>, angularLimitsMin=<-5 -35 -25>, angularTarget=<0 0 0>}
      localJointOffset=<2 0 0>
     }]
    relativeSpaceBone {boneName=root}
    sphericalLimits [{limitRadius=20, drivingBone {boneName=clavicle_l}, sphereLocalOffset=<15 -10 -7>}]
   }
   n22 = Anim Dynamics - Bone: dyn_braidouter_C_02_l(ComponentPose=@n21, Alpha=1, ExternalForce="0,0,0") {
    angularDampingOverride=0.9999
    angularSpringConstant=75
    bAngularSpring=true
    bGravityOverrideInSimSpace=true
    bOverrideAngularDamping=true
    bUseSphericalLimits=true
    simulationSpace=RootRelative
    alphaBoolBlend {blendOption=Linear}
    boundBone {boneName=dyn_braidouter_C_02_l}
    componentAppliedLinearAccClamp=<3500 3500 3500>
    componentLinearAccScale=<0.9 0.9 0.9>
    componentLinearVelScale=<0.9 0.9 0.9>
    physicsBodyDefinitions [{
      collisionType=CustomSphere
      sphereCollisionRadius=4
      boundBone {boneName=dyn_braidouter_C_02_l}
      boxExtents=<25 15 15>
      constraintSetup {angularLimitsMax=<5 10 25>, angularLimitsMin=<-5 -35 -25>, angularTarget=<0 0 0>}
      localJointOffset=<2 0 0>
     }]
    relativeSpaceBone {boneName=root}
    sphericalLimits [{limitRadius=20, drivingBone {boneName=clavicle_l}, sphereLocalOffset=<15 -10 -7>}]
   }
   n23 = Anim Dynamics - Bone: dyn_braidouter_C_03_l(ComponentPose=@n22, Alpha=1, ExternalForce="0,0,0") {
    angularDampingOverride=0.9999
    angularSpringConstant=75
    bAngularSpring=true
    bGravityOverrideInSimSpace=true
    bOverrideAngularDamping=true
    bUseSphericalLimits=true
    simulationSpace=RootRelative
    alphaBoolBlend {blendOption=Linear}
    boundBone {boneName=dyn_braidouter_C_03_l}
    componentAppliedLinearAccClamp=<3500 3500 3500>
    componentLinearAccScale=<0.9 0.9 0.9>
    componentLinearVelScale=<0.9 0.9 0.9>
    physicsBodyDefinitions [{
      collisionType=CustomSphere
      sphereCollisionRadius=4
      boundBone {boneName=dyn_braidouter_C_03_l}
      boxExtents=<25 15 15>
      constraintSetup {angularLimitsMax=<5 20 25>, angularLimitsMin=<-5 -35 -25>, angularTarget=<0 0 0>}
      localJointOffset=<2 0 0>
     }]
    relativeSpaceBone {boneName=root}
    sphericalLimits [{limitRadius=20, drivingBone {boneName=clavicle_l}, sphereLocalOffset=<15 -10 -7>}]
   }
   n24 = Anim Dynamics - Bone: dyn_braidouter_C_04_l(ComponentPose=@n23, Alpha=1, ExternalForce="0,0,0") {
    angularDampingOverride=0.9999
    angularSpringConstant=75
    bAngularSpring=true
    bGravityOverrideInSimSpace=true
    bOverrideAngularDamping=true
    bUseSphericalLimits=true
    simulationSpace=RootRelative
    alphaBoolBlend {blendOption=Linear}
    boundBone {boneName=dyn_braidouter_C_04_l}
    componentAppliedLinearAccClamp=<3500 3500 3500>
    componentLinearAccScale=<0.9 0.9 0.9>
    componentLinearVelScale=<0.9 0.9 0.9>
    physicsBodyDefinitions [{
      collisionType=CustomSphere
      sphereCollisionRadius=4
      boundBone {boneName=dyn_braidouter_C_04_l}
      boxExtents=<25 15 15>
      constraintSetup {angularLimitsMax=<5 35 25>, angularLimitsMin=<-5 -35 -25>, angularTarget=<0 0 0>}
      localJointOffset=<2 0 0>
     }]
    relativeSpaceBone {boneName=root}
    sphericalLimits [{limitRadius=20, drivingBone {boneName=clavicle_l}, sphereLocalOffset=<15 -10 -7>}]
   }
   n25 = Anim Dynamics - Bone: dyn_braidouter_B_00_l(ComponentPose=@n24, Alpha=1, ExternalForce="0,0,0") {
    angularDampingOverride=0.9999
    angularSpringConstant=75
    bAngularSpring=true
    bGravityOverrideInSimSpace=true
    bOverrideAngularDamping=true
    bUseSphericalLimits=true
    simulationSpace=RootRelative
    alphaBoolBlend {blendOption=Linear}
    boundBone {boneName=dyn_braidouter_B_00_l}
    componentAppliedLinearAccClamp=<3500 3500 3500>
    componentLinearAccScale=<0.9 0.9 0.9>
    componentLinearVelScale=<0.9 0.9 0.9>
    physicsBodyDefinitions [{
      collisionType=CustomSphere
      sphereCollisionRadius=4
      boundBone {boneName=dyn_braidouter_B_00_l}
      boxExtents=<25 15 15>
      constraintSetup {angularLimitsMax=<5 0 25>, angularLimitsMin=<-5 -35 -25>, angularTarget=<0 0 0>}
      localJointOffset=<2 0 0>
     }]
    relativeSpaceBone {boneName=root}
    sphericalLimits [{limitRadius=6, drivingBone {boneName=clavicle_l}, sphereLocalOffset=<0 0 -3>}]
   }
   n26 = Anim Dynamics - Bone: dyn_braidouter_B_01_l(ComponentPose=@n25, Alpha=1, ExternalForce="0,0,0") {
    angularDampingOverride=0.9999
    angularSpringConstant=75
    bAngularSpring=true
    bGravityOverrideInSimSpace=true
    bOverrideAngularDamping=true
    bUseSphericalLimits=true
    simulationSpace=RootRelative
    alphaBoolBlend {blendOption=Linear}
    boundBone {boneName=dyn_braidouter_B_01_l}
    componentAppliedLinearAccClamp=<3500 3500 3500>
    componentLinearAccScale=<0.9 0.9 0.9>
    componentLinearVelScale=<0.9 0.9 0.9>
    physicsBodyDefinitions [{
      collisionType=CustomSphere
      sphereCollisionRadius=4
      boundBone {boneName=dyn_braidouter_B_01_l}
      boxExtents=<25 15 15>
      constraintSetup {angularLimitsMax=<5 5 25>, angularLimitsMin=<-5 -35 -25>, angularTarget=<0 0 0>}
      localJointOffset=<2 0 0>
     }]
    relativeSpaceBone {boneName=root}
    sphericalLimits [{limitRadius=20, drivingBone {boneName=clavicle_l}, sphereLocalOffset=<15 -10 -7>}]
   }
   n27 = Anim Dynamics - Bone: dyn_braidouter_B_02_l(ComponentPose=@n26, Alpha=1, ExternalForce="0,0,0") {
    angularDampingOverride=0.9999
    angularSpringConstant=75
    bAngularSpring=true
    bGravityOverrideInSimSpace=true
    bOverrideAngularDamping=true
    bUseSphericalLimits=true
    simulationSpace=RootRelative
    alphaBoolBlend {blendOption=Linear}
    boundBone {boneName=dyn_braidouter_B_02_l}
    componentAppliedLinearAccClamp=<3500 3500 3500>
    componentLinearAccScale=<0.9 0.9 0.9>
    componentLinearVelScale=<0.9 0.9 0.9>
    physicsBodyDefinitions [{
      collisionType=CustomSphere
      sphereCollisionRadius=4
      boundBone {boneName=dyn_braidouter_B_02_l}
      boxExtents=<25 15 15>
      constraintSetup {angularLimitsMax=<5 10 25>, angularLimitsMin=<-5 -35 -25>, angularTarget=<0 0 0>}
      localJointOffset=<2 0 0>
     }]
    relativeSpaceBone {boneName=root}
    sphericalLimits [{limitRadius=20, drivingBone {boneName=clavicle_l}, sphereLocalOffset=<15 -10 -7>}]
   }
   n28 = Anim Dynamics - Bone: dyn_braidouter_B_03_l(ComponentPose=@n27, Alpha=1, ExternalForce="0,0,0") {
    angularDampingOverride=0.9999
    angularSpringConstant=75
    bAngularSpring=true
    bGravityOverrideInSimSpace=true
    bOverrideAngularDamping=true
    bUseSphericalLimits=true
    simulationSpace=RootRelative
    alphaBoolBlend {blendOption=Linear}
    boundBone {boneName=dyn_braidouter_B_03_l}
    componentAppliedLinearAccClamp=<3500 3500 3500>
    componentLinearAccScale=<0.9 0.9 0.9>
    componentLinearVelScale=<0.9 0.9 0.9>
    physicsBodyDefinitions [{
      collisionType=CustomSphere
      sphereCollisionRadius=4
      boundBone {boneName=dyn_braidouter_B_03_l}
      boxExtents=<25 15 15>
      constraintSetup {angularLimitsMax=<5 20 25>, angularLimitsMin=<-5 -35 -25>, angularTarget=<0 0 0>}
      localJointOffset=<2 0 0>
     }]
    relativeSpaceBone {boneName=root}
    sphericalLimits [{limitRadius=20, drivingBone {boneName=clavicle_l}, sphereLocalOffset=<15 -10 -7>}]
   }
   n29 = Anim Dynamics - Bone: dyn_braidouter_B_04_l(ComponentPose=@n28, Alpha=1, ExternalForce="0,0,0") {
    angularDampingOverride=0.9999
    angularSpringConstant=75
    bAngularSpring=true
    bGravityOverrideInSimSpace=true
    bOverrideAngularDamping=true
    bUseSphericalLimits=true
    simulationSpace=RootRelative
    alphaBoolBlend {blendOption=Linear}
    boundBone {boneName=dyn_braidouter_B_04_l}
    componentAppliedLinearAccClamp=<3500 3500 3500>
    componentLinearAccScale=<0.9 0.9 0.9>
    componentLinearVelScale=<0.9 0.9 0.9>
    physicsBodyDefinitions [{
      collisionType=CustomSphere
      sphereCollisionRadius=4
      boundBone {boneName=dyn_braidouter_B_04_l}
      boxExtents=<25 15 15>
      constraintSetup {angularLimitsMax=<5 35 25>, angularLimitsMin=<-5 -35 -25>, angularTarget=<0 0 0>}
      localJointOffset=<2 0 0>
     }]
    relativeSpaceBone {boneName=root}
    sphericalLimits [{limitRadius=20, drivingBone {boneName=clavicle_l}, sphereLocalOffset=<15 -10 -7>}]
   }
   n30 = Anim Dynamics - Bone: dyn_braidouter_A_00_l(ComponentPose=@n29, Alpha=1, ExternalForce="0,0,0") {
    angularDampingOverride=0.9999
    angularSpringConstant=75
    bAngularSpring=true
    bGravityOverrideInSimSpace=true
    bOverrideAngularDamping=true
    bUseSphericalLimits=true
    simulationSpace=RootRelative
    alphaBoolBlend {blendOption=Linear}
    boundBone {boneName=dyn_braidouter_A_00_l}
    componentAppliedLinearAccClamp=<3500 3500 3500>
    componentLinearAccScale=<0.9 0.9 0.9>
    componentLinearVelScale=<0.9 0.9 0.9>
    physicsBodyDefinitions [{
      collisionType=CustomSphere
      sphereCollisionRadius=4
      boundBone {boneName=dyn_braidouter_A_00_l}
      boxExtents=<25 15 15>
      constraintSetup {angularLimitsMax=<5 0 25>, angularLimitsMin=<-5 -35 -25>, angularTarget=<0 0 0>}
      localJointOffset=<2 0 0>
     }]
    relativeSpaceBone {boneName=root}
    sphericalLimits [{limitRadius=6, drivingBone {boneName=clavicle_l}, sphereLocalOffset=<0 0 -3>}]
   }
   n31 = Anim Dynamics - Bone: dyn_braidouter_A_01_l(ComponentPose=@n30, Alpha=1, ExternalForce="0,0,0") {
    angularDampingOverride=0.9999
    angularSpringConstant=75
    bAngularSpring=true
    bGravityOverrideInSimSpace=true
    bOverrideAngularDamping=true
    bUseSphericalLimits=true
    simulationSpace=RootRelative
    alphaBoolBlend {blendOption=Linear}
    boundBone {boneName=dyn_braidouter_A_01_l}
    componentAppliedLinearAccClamp=<3500 3500 3500>
    componentLinearAccScale=<0.9 0.9 0.9>
    componentLinearVelScale=<0.9 0.9 0.9>
    physicsBodyDefinitions [{
      collisionType=CustomSphere
      sphereCollisionRadius=4
      boundBone {boneName=dyn_braidouter_A_01_l}
      boxExtents=<25 15 15>
      constraintSetup {angularLimitsMax=<5 5 25>, angularLimitsMin=<-5 -35 -25>, angularTarget=<0 0 0>}
      localJointOffset=<2 0 0>
     }]
    relativeSpaceBone {boneName=root}
    sphericalLimits [{limitRadius=20, drivingBone {boneName=clavicle_l}, sphereLocalOffset=<15 -10 -7>}]
   }
   n32 = Anim Dynamics - Bone: dyn_braidouter_A_02_l(ComponentPose=@n31, Alpha=1, ExternalForce="0,0,0") {
    angularDampingOverride=0.9999
    angularSpringConstant=75
    bAngularSpring=true
    bGravityOverrideInSimSpace=true
    bOverrideAngularDamping=true
    bUseSphericalLimits=true
    simulationSpace=RootRelative
    alphaBoolBlend {blendOption=Linear}
    boundBone {boneName=dyn_braidouter_A_02_l}
    componentAppliedLinearAccClamp=<3500 3500 3500>
    componentLinearAccScale=<0.9 0.9 0.9>
    componentLinearVelScale=<0.9 0.9 0.9>
    physicsBodyDefinitions [{
      collisionType=CustomSphere
      sphereCollisionRadius=4
      boundBone {boneName=dyn_braidouter_A_02_l}
      boxExtents=<25 15 15>
      constraintSetup {angularLimitsMax=<5 10 25>, angularLimitsMin=<-5 -35 -25>, angularTarget=<0 0 0>}
      localJointOffset=<2 0 0>
     }]
    relativeSpaceBone {boneName=root}
    sphericalLimits [{limitRadius=20, drivingBone {boneName=clavicle_l}, sphereLocalOffset=<15 -10 -7>}]
   }
   n33 = Anim Dynamics - Bone: dyn_braidouter_A_03_l(ComponentPose=@n32, Alpha=1, ExternalForce="0,0,0") {
    angularDampingOverride=0.9999
    angularSpringConstant=75
    bAngularSpring=true
    bGravityOverrideInSimSpace=true
    bOverrideAngularDamping=true
    bUseSphericalLimits=true
    simulationSpace=RootRelative
    alphaBoolBlend {blendOption=Linear}
    boundBone {boneName=dyn_braidouter_A_03_l}
    componentAppliedLinearAccClamp=<3500 3500 3500>
    componentLinearAccScale=<0.9 0.9 0.9>
    componentLinearVelScale=<0.9 0.9 0.9>
    physicsBodyDefinitions [{
      collisionType=CustomSphere
      sphereCollisionRadius=4
      boundBone {boneName=dyn_braidouter_A_03_l}
      boxExtents=<25 15 15>
      constraintSetup {angularLimitsMax=<5 20 25>, angularLimitsMin=<-5 -35 -25>, angularTarget=<0 0 0>}
      localJointOffset=<2 0 0>
     }]
    relativeSpaceBone {boneName=root}
    sphericalLimits [{limitRadius=20, drivingBone {boneName=clavicle_l}, sphereLocalOffset=<15 -10 -7>}]
   }
   n34 = Anim Dynamics - Bone: dyn_braidouter_A_04_l(ComponentPose=@n33, Alpha=1, ExternalForce="0,0,0") {
    angularDampingOverride=0.9999
    angularSpringConstant=75
    bAngularSpring=true
    bGravityOverrideInSimSpace=true
    bOverrideAngularDamping=true
    bUseSphericalLimits=true
    simulationSpace=RootRelative
    alphaBoolBlend {blendOption=Linear}
    boundBone {boneName=dyn_braidouter_A_04_l}
    componentAppliedLinearAccClamp=<3500 3500 3500>
    componentLinearAccScale=<0.9 0.9 0.9>
    componentLinearVelScale=<0.9 0.9 0.9>
    physicsBodyDefinitions [{
      collisionType=CustomSphere
      sphereCollisionRadius=4
      boundBone {boneName=dyn_braidouter_A_04_l}
      boxExtents=<25 15 15>
      constraintSetup {angularLimitsMax=<5 35 25>, angularLimitsMin=<-5 -35 -25>, angularTarget=<0 0 0>}
      localJointOffset=<2 0 0>
     }]
    relativeSpaceBone {boneName=root}
    sphericalLimits [{limitRadius=20, drivingBone {boneName=clavicle_l}, sphereLocalOffset=<15 -10 -7>}]
   }
   n35 = "Transform (Modify) Bone - Bone: lazer_01"(ComponentPose=@n34, Alpha=Get headAimAlpha() -> (headAimAlpha:double), Translation="0,0,0", Rotation="50,0.000,0.000", Scale="1,1,1") {
    rotationMode=BMM_Additive
    rotationSpace=BCS_BoneSpace
    alphaBoolBlend {blendOption=Linear}
    alphaScaleBiasClamp {bInterpResult=true}
    boneToModify {boneName=lazer_01}
    rotation {pitch=50}
   }
   Output Pose(Result=Component To Local(ComponentPose=@n35)) {layerGroup=None}
  ]}
 EventGraph {nodes [
   Event BlueprintInitializeAnimation()
   > Set Body_Mesh(Body_Mesh=GetOwnerSkeletalMeshComponent())
   Event OnCharacterCustomizationLoaded()
   > Do Once()
   n1 = CheckForBackbling()
   n2 = NOT Boolean(A=@n1.bIsBackblingLong)
   > Set isWearingBackbling(isWearingBackbling:double="To Float (Boolean)"(InBool=AND Boolean(A=@n1, B=@n2))) -> (Output_Get:double=0)
   > Set isWearingCape(isWearingCape:double="To Float (Boolean)"(InBool=@n1.bIsBackblingLong)) -> (Output_Get:double)
   Event BlueprintUpdateAnimation()
   > n3 = Sequence()
   n4 = GetOwnerSkeletalMeshComponent()
   > Set headAimAlpha(headAimAlpha:double=SelectFloat(
     A=GetAlphaFromReferenceTransformDelta(
      SkeletalMeshComponent=@n4
      BoneName=head
      BaseName=neck_02
      DeltaComponentType=EulerY
      InputRange=(X=-5,Y=15)
      OutputRange=(X=0,Y=0.7)
      TwistAxis="1,0,0"
     )
     B=0
     bPickA=Get bIsTargeting()
    )) -> (Output_Get:double)
   n5 = Break Fort Anim Input Skydiving External Force(FortAnimInput_SkydivingExternalForce=Get SkydivingExternalForceInput()) -> (
    bUseSkydivingVectorForce=~
    SkydivingVectorForce=~
    HeadToPelvisDirection=~
    FloatingMultiplier=~
    FloatingAdditive=~
    DivingMultiplier=~
    DivingAdditive=~
    ParachutingMultiplier=~
    ParachutingAdditive=~
   )
   Set windVector(
    execute=@n3.then_1
    windVector:Vector=SelectVector(A=vector * float(A=@n5.SkydivingVectorForce, B=0.2), B="0, 0,-980", bPickA=Get bIsSkyDiving())
   ) -> (Output_Get:Vector)
  ]}
]
variables [
 armSwingDelta_r {type=number}
 Body_Mesh {type=SkeletalMeshComponent}
 headAimAlpha {type=number}
 isWearingBackbling {type=number}
 isWearingCape {type=number}
 windVector {type=Vector}
]
