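# Post-process AnimBP for the Nightmare skin's face accessory, posing the mandible bones around mask state and emotes.
# Four chained per-bone blends layer mask-off, mask-on, generic-emote and facial idle sequences over the input pose.
# The Nightmare emote montage's position scrubs the mask sequences, switching from mask-on to mask-off at three seconds.
# Other emote montages, found by name, trigger the generic pose; idle requires no emote and an inactive Stage1 Parts variant.
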AnimBlueprint /BRCosmetics/Characters/Player/Male/Medium/Bodies/M_MED_Nightmare/Meshes/Parts/M_MED_Nightmare_FaceAcc_PPAnimBP
target_skeleton=/Game/Characters/Player/Male/Medium/Bodies/M_MED_Nightmare/Meshes/Parts/M_MED_Nightmare_Skeleton
graphs [
 AnimGraph {nodes [
   n1 = Layered blend per bone(
    BasePose=Input Pose()
    BlendPoses_0="Sequence Evaluator (by time) 'Nightmare_FaceAcc_MaskOff_Mandibles'"!(ExplicitTime:float=Get MaskOffPosition() -> (MaskOffPosition:double)) -> (Pose:Pose Link Structure) {
     is=title AnimGraphNode_SequenceEvaluator
     sequence=/BRCosmetics/Animation/Game/MainPlayer/Emotes/Nightmare/Nightmare_FaceAcc_MaskOff_Mandibles
    }
    BlendWeights_0=Get MaskOffBlend() -> (MaskOffBlend:double)
   ) {layerSetup [{branchFilters [
       {boneName=face_01_l}
       {boneName=face_02_l}
       {boneName=face_03_l}
       {boneName=face_01_r}
       {boneName=face_02_r}
       {boneName=face_03_r}
       {boneName=face_04_l}
       {boneName=face_05_l}
       {boneName=face_06_l}
       {boneName=face_04_r}
       {boneName=face_05_r}
       {boneName=face_06_r}
      ]}]}
   n2 = Layered blend per bone(
    BasePose=@n1
    BlendPoses_0="Sequence Evaluator (by time) 'Nightmare_FaceAcc_MaskOn_Mandibles'"!(ExplicitTime:float=Get MaskOnPosition() -> (MaskOnPosition:double)) -> (Pose:Pose Link Structure) {
     is=title AnimGraphNode_SequenceEvaluator
     sequence=/BRCosmetics/Animation/Game/MainPlayer/Emotes/Nightmare/Nightmare_FaceAcc_MaskOn_Mandibles
    }
    BlendWeights_0=Get MaskOnBlend() -> (MaskOnBlend:double)
   ) {layerSetup [{branchFilters [
       {boneName=face_01_l}
       {boneName=face_02_l}
       {boneName=face_03_l}
       {boneName=face_01_r}
       {boneName=face_02_r}
       {boneName=face_03_r}
       {boneName=face_04_l}
       {boneName=face_05_l}
       {boneName=face_06_l}
       {boneName=face_04_r}
       {boneName=face_05_r}
       {boneName=face_06_r}
      ]}]}
   n3 = Layered blend per bone(
    BasePose=@n2
    BlendPoses_0=Sequence Player 'Nightmare_FaceAcc_MaskOn_Mandibles'() {sequence=&Nightmare_FaceAcc_MaskOn_Mandibles}
    BlendWeights_0=Get genericEmoteBlend() -> (genericEmoteBlend:double)
   ) {layerSetup [{branchFilters [
       {boneName=face_01_l}
       {boneName=face_02_l}
       {boneName=face_03_l}
       {boneName=face_01_r}
       {boneName=face_02_r}
       {boneName=face_03_r}
       {boneName=face_04_l}
       {boneName=face_05_l}
       {boneName=face_06_l}
       {boneName=face_04_r}
       {boneName=face_05_r}
       {boneName=face_06_r}
      ]}]}
   n4 = Layered blend per bone(
    BasePose=@n3
    BlendPoses_0=Sequence Player 'Nightmare_Facial_Idle_Mandibles'() {sequence=/BRCosmetics/Animation/Game/MainPlayer/Facial/Nightmare/Nightmare_Facial_Idle_Mandibles}
    BlendWeights_0=Get idleBlend() -> (idleBlend:double)
   ) {layerSetup [{branchFilters [
       {boneName=face_01_l}
       {boneName=face_02_l}
       {boneName=face_03_l}
       {boneName=face_01_r}
       {boneName=face_02_r}
       {boneName=face_03_r}
       {boneName=face_04_l}
       {boneName=face_05_l}
       {boneName=face_06_l}
       {boneName=face_04_r}
       {boneName=face_05_r}
       {boneName=face_06_r}
      ]}]}
   Output Pose(Result=@n4) {layerGroup=None}
  ]}
 EventGraph {nodes [
   Event BlueprintUpdateAnimation()
   > n1 = Sequence()
   > n2 = Cast To FortPlayerPawn(Object=GetOwningActor())
   > n3 = GetVariantChannelActiveIndexForItemActor!(
    execute
    Item:Athena Cosmetic Item Definition Object Reference=/BRCosmetics/Athena/Items/Cosmetics/Characters/CID_991_Athena_Commando_M_Nightmare_NM1C8
    Actor:Actor Object Reference=@n2
    Channel:Gameplay Tag Structure=(TagName="Cosmetics.Variant.Channel.Parts")
    VariantList:Array of Gameplay Tag Structures=Make Array("[0]"=(TagName="Cosmetics.Variant.Property.Stage1"))
    bErrorIfChannelMissing:bool=false
    bErrorIfTagNotInArray:bool=false
   ) -> (then, Index:int=0, bIndexIsValid:bool, bFoundChannel:bool)
   > n4 = Branch(Condition=@n3.bIndexIsValid)
   > Set maskIsOn(maskIsOn:bool=true) -> (Output_Get:bool)
   Set maskIsOn(execute=@n4.else, maskIsOn=false)
   Set MaskOnPosition(execute=@n1.then_1, MaskOnPosition:double=0) -> (Output_Get:double)
   > Set MaskOffPosition(MaskOffPosition:double=0) -> (Output_Get:double)
   n5 = GetAnimInstance(self=GetOwnerSkeletalMeshComponent())
   n6 = GetCurrentActiveMontage!(self:Anim Instance Object Reference=@n5) -> (ReturnValue:Anim Montage Object Reference)
   n7 = "Not Equal (Object)"!(A:Object Reference=@n6, B:Object Reference=/BRCosmetics/Animation/Game/MainPlayer/Emotes/Nightmare/Emote_Nightmare_CMM_M) -> (ReturnValue:bool)
   n8 = "Equal (Integer)"(A=FindSubstring(SearchIn="To String (Object)"!(InObj:Object Reference=@n6) -> (ReturnValue:String), Substring=emote, StartPosition=0), B=0) {is=title K2Node_CallFunction}
   > Set genericEmoteBlend(genericEmoteBlend:double=SelectFloat(A=1, B=0, bPickA=AND Boolean(A=@n8, B=@n7))) -> (Output_Get:double)
   n9 = NOT Boolean(A=Get maskIsOn() -> (maskIsOn:bool))
   n10 = NOT Boolean(A=Get bIsPlayingEmote())
   > Set idleBlend(idleBlend:double=SelectFloat(A=0.7, B=0, bPickA=AND Boolean!(A=@n7, B=@n10, C:bool=@n9) -> (ReturnValue))) -> (Output_Get:double)
   n11 = Montage_GetPosition!(
    self:Anim Instance Object Reference=@n5
    Montage:Anim Montage Object Reference=&Emote_Nightmare_CMM_M
   ) -> (ReturnValue:float)
   n12 = "float < float"!(A:double=@n11, B:double=3) -> (ReturnValue:bool) {is=title K2Node_CallFunction}
   > Set MaskOffBlend(MaskOffBlend:double=SelectFloat(A=0, B=1, bPickA=@n12)) -> (Output_Get:double)
   > Set MaskOnBlend(MaskOnBlend:double=SelectFloat(A=1, B=0, bPickA=@n12)) -> (Output_Get:double)
   > Set MaskOffPosition(MaskOffPosition=float - float(A=@n11, B=3) {is=title K2Node_CallFunction})
   > Set MaskOnPosition(MaskOnPosition=@n11)
   Event BlueprintInitializeAnimation()
   > Set maskIsOn(maskIsOn=true)
   SelectFloat(A=0.7, B=1, bPickA=Get bIsInFrontEnd())
  ]}
]
variables [
 animInst {type=AnimInstance, default=null}
 currentHelmet {type=AthenaBackpackItemDefinition, default=null}
 genericEmoteBlend {type=number}
 headAnimBP {type=FortAnimationBlueprintLibrary, default=null}
 headSkelMesh {type=SkeletalMesh, default=/BRCosmetics/Characters/Player/Male/Medium/Heads/M_MED_Nightmare_Head/Meshes/Male_Medium_Nightmare_Head}
 helmetItemDef {type=AthenaBackpackItemDefinition, default=null}
 idleBlend {type=number}
 maskIsOn {type=boolean}
 MaskOffBlend {type=number}
 MaskOffPosition {type=number}
 MaskOnBlend {type=number}
 MaskOnPosition {type=number}
 montageIsNotPlaying {type=boolean}
 NewVar_0 {type=CustomCharacterPartAnimInstance, default=null}
 NewVar_1 {type=CustomCharacterPartAnimInstance, default=null}
 NewVar_10 {type=boolean}
 NewVar_11 {type=AnimMontage, default=null}
 NewVar_12 {type=AnimMontage, default=null}
 NewVar_13 {type=boolean}
 NewVar_14 {type=string, default=None}
 NewVar_15 {type=boolean}
 NewVar_16 {type=boolean}
 NewVar_17 {type=boolean}
 NewVar_18 {type=Object, default=null}
 NewVar_19 {type="array<McpVariantChannelInfo>", default []}
 NewVar_2 {type=AnimMontage, default=null}
 NewVar_20 {type=boolean}
 NewVar_21 {
  type=McpVariantChannelInfo
  default {
   customData=""
   itemVariantIsUsedFor=null
   activeVariantTag {tagName=None}
   ownedVariantTags {gameplayTags []}
   variantChannelTag {tagName=None}
  }
 }
 NewVar_22 {type="array<McpVariantChannelInfo>", default []}
 NewVar_23 {
  type=McpVariantChannelInfo
  default {
   customData=""
   itemVariantIsUsedFor=null
   activeVariantTag {tagName=None}
   ownedVariantTags {gameplayTags []}
   variantChannelTag {tagName=None}
  }
 }
 NewVar_24 {default {}}
 NewVar_25 {type=GameplayTag, default {tagName=None}}
 NewVar_26 {type="array<GameplayTag>", default []}
 NewVar_27 {type="array<GameplayTag>", default [{tagName=Cosmetics.Variant.Property}]}
 NewVar_28 {type=integer}
 NewVar_29 {type=FortPlayerPawn}
 NewVar_3 {type=SkeletalMeshComponent}
 NewVar_30 {type=boolean}
 NewVar_4 {type=Character, default=null}
 NewVar_5 {type=CustomCharacterPartAnimInstance, default=null}
 NewVar_6 {type=SkeletalMeshComponent}
 NewVar_7 {type=AnimInstance, default=null}
 NewVar_8 {type=AnimMontage, default=null}
 NewVar_9 {type=AnimMontage, default=null}
 nightmareHat {type=SkeletalMesh, default=/BRCosmetics/Animation/Props/Nightmare_Mask/Meshes/Nightmare_Mask_Gadget}
 pawn {type=FortPlayerPawn}
 skelMeshHead {type=SkeletalMeshComponent}
 variantChannelValue {
  type=McpVariantChannelInfo
  default {
   customData=""
   itemVariantIsUsedFor=null
   activeVariantTag {tagName=None}
   ownedVariantTags {gameplayTags []}
   variantChannelTag {tagName=None}
  }
 }
]
